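# Corrective Control Rig for the F_MED_ForwardLake_Dire female body, posing dyn_ flap, sash and sleeve-fix bones from body pose.
# Arm, thigh and spine rotations plus pelvis-to-foot and pelvis-to-root distances are remapped into alpha variables.
# These weight parent constraints to thigh and upper-arm twist bones, with smoothed cross-leg blends for front flaps and sash.
# A smoothed low-pelvis, tucked-foot alpha also drives a transform offset on bones including both back flaps.
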ControlRigBlueprint /BRCosmetics/Characters/Player/Female/Medium/Bodies/F_MED_ForwardLake_Dire/Meshes/F_MED_ForwardLake_Dire_CtrlRig
graphs [
 RigVMFunctionLibrary {links [], nodes []}
 RigVMModel {nodes [
   RigUnit_BeginExecution = &RigUnit_BeginExecution()
   n1 = &RigUnit_GetRelativeTransformForItem(Child=(Type=Bone,Name="upperarm_l"), Parent=(Type=Bone,Name="spine_05"), bChildInitial=False, bParentInitial=False)
   n2 = &RigVMFunction_MathQuaternionToEuler(RotationOrder=XZY, Value=@n1.RelativeTransform.Rotation)
   RigVMVariableNode(
    Value=&RigVMFunction_MathDoubleRemap(SourceMaximum=-10, SourceMinimum=10, TargetMaximum=1, TargetMinimum=0, Value=@n2.Result.Z).Result
    ExecuteContext=@RigUnit_BeginExecution.ExecutePin
   ) {variable=L upperarmBack alpha}
   > RigVMVariableNode(Value=&RigVMFunction_MathDoubleRemap(SourceMaximum=40, SourceMinimum=0, TargetMaximum=1, TargetMinimum=0, Value=@n2.Result.Z).Result) {variable=L upperarmForward alpha}
   > RigVMVariableNode(Value=&RigVMFunction_MathDoubleRemap(
     SourceMaximum=30
     SourceMinimum=0
     TargetMaximum=1
     TargetMinimum=0
     Value=&RigVMFunction_MathQuaternionToEuler(RotationOrder=XYZ, Value=@n1.RelativeTransform.Rotation).Result.Y
    ).Result) {variable=L upperarmUp alpha}
   n3 = &RigUnit_GetRelativeTransformForItem(Child=(Type=Bone,Name="upperarm_r"), Parent=(Type=Bone,Name="spine_05"), bChildInitial=False, bParentInitial=False)
   n4 = &RigVMFunction_MathQuaternionToEuler(RotationOrder=XZY, Value=@n3.RelativeTransform.Rotation)
   > RigVMVariableNode(Value=&RigVMFunction_MathDoubleRemap(SourceMaximum=10, SourceMinimum=-10, TargetMaximum=1, TargetMinimum=0, Value=@n4.Result.Z).Result) {variable=R upperarmBack alpha}
   > RigVMVariableNode(Value=&RigVMFunction_MathDoubleRemap(SourceMaximum=-40, SourceMinimum=0, TargetMaximum=1, TargetMinimum=0, Value=@n4.Result.Z).Result) {variable=R upperarmForward alpha}
   > RigVMVariableNode(Value=&RigVMFunction_MathDoubleRemap(
     SourceMaximum=30
     SourceMinimum=0
     TargetMaximum=1
     TargetMinimum=0
     Value=&RigVMFunction_MathQuaternionToEuler(RotationOrder=XYZ, Value=@n3.RelativeTransform.Rotation).Result.Y
    ).Result) {variable=R upperarmUp alpha}
   n5 = &RigVMFunction_MathQuaternionToEuler(
    RotationOrder=ZXY
    Value=&RigUnit_GetRelativeTransformForItem(Child=(Type=Bone,Name="thigh_r"), Parent=(Type=Bone,Name="pelvis"), bChildInitial=False, bParentInitial=False).RelativeTransform.Rotation
   )
   > RigVMVariableNode(Value=&RigVMFunction_MathDoubleRemap(SourceMaximum=-7, SourceMinimum=10, TargetMaximum=1, TargetMinimum=0, Value=@n5.Result.Y).Result) {variable=R thighInward alpha}
   Absolute = &RigVMFunction_MathDoubleAbs(Value=@n5.Result.Z)
   Subtract = &RigVMFunction_MathDoubleSub(A=180, B=@Absolute.Result)
   > RigVMVariableNode(Value=&RigVMFunction_MathDoubleRemap(SourceMaximum=50, SourceMinimum=15, TargetMaximum=1, TargetMinimum=0, Value=@Subtract.Result).Result) {variable=R thighForward alpha}
   > RigVMVariableNode(Value=&RigVMFunction_MathDoubleRemap(
     SourceMaximum=55
     SourceMinimum=10
     TargetMaximum=1
     TargetMinimum=0
     Value=&RigVMFunction_MathDoubleAbs(Value=@n5.Result.Y).Result
    ).Result) {variable=R thighVeryOutwards alpha}
   n6 = &RigVMFunction_MathQuaternionToEuler(
    RotationOrder=ZXY
    Value=&RigUnit_GetRelativeTransformForItem(Child=(Type=Bone,Name="thigh_l"), Parent=(Type=Bone,Name="pelvis"), bChildInitial=False, bParentInitial=False).RelativeTransform.Rotation
   )
   > RigVMVariableNode(Value=&RigVMFunction_MathDoubleRemap(SourceMaximum=-7, SourceMinimum=10, TargetMaximum=0, TargetMinimum=1, Value=@n6.Result.Y).Result) {variable=L thighInward alpha}
   > RigVMVariableNode(Value=&RigVMFunction_MathDoubleRemap(SourceMaximum=-50, SourceMinimum=-15, TargetMaximum=1, TargetMinimum=0, Value=@n6.Result.Z).Result) {variable=L thighForward alpha}
   > RigVMVariableNode(Value=&RigVMFunction_MathDoubleRemap(
     SourceMaximum=55
     SourceMinimum=10
     TargetMaximum=1
     TargetMinimum=0
     Value=&RigVMFunction_MathDoubleAbs(Value=@n6.Result.Y).Result
    ).Result) {variable=L thighVeryOutwards alpha}
   n7 = &RigUnit_GetRelativeTransformForItem(Child=(Type=Bone,Name="spine_04"), Parent=(Type=Bone,Name="pelvis"), bChildInitial=False, bParentInitial=False)
   > RigVMVariableNode(Value=&RigVMFunction_MathDoubleRemap(
     SourceMaximum=15
     SourceMinimum=0
     TargetMaximum=1
     TargetMinimum=0
     Value=&RigVMFunction_MathQuaternionToEuler(RotationOrder=ZXY, Value=@n7.RelativeTransform.Rotation).Result.X
    ).Result) {variable=spineTwist Rightwards}
   n8 = &RigUnit_GetTransform(Item=(Type=Bone,Name="pelvis"), bInitial=False)
   n9 = &RigVMFunction_MathVectorDistance(B=&RigUnit_GetTransform(Item=(Type=Bone,Name="foot_l"), bInitial=False).Transform.Translation, A=@n8.Transform.Translation)
   n10 = &RigVMFunction_MathVectorDistance(B=&RigUnit_GetTransform(Item=(Type=Bone,Name="foot_r"), bInitial=False).Transform.Translation, A=@n8.Transform.Translation)
   > RigVMVariableNode(Value=&RigVMFunction_MathDoubleMax(
     A=&RigVMFunction_MathDoubleRemap(SourceMaximum=29, SourceMinimum=34, TargetMaximum=1, TargetMinimum=0, Value=@n9.Result).Result
     B=&RigVMFunction_MathDoubleRemap(SourceMaximum=29, SourceMinimum=34, TargetMaximum=1, TargetMinimum=0, Value=@n10.Result).Result
    ).Result) {variable=eitherFootCloseToPelvis}
   n11 = &RigUnit_GetTransform(Item=(Type=Bone,Name="root"), bInitial=False)
   n12 = &RigUnit_GetTransform(Item=(Type=Bone,Name="pelvis"), bInitial=False)
   > RigVMVariableNode(Value=&RigVMFunction_MathDoubleRemap(
     SourceMaximum=25
     SourceMinimum=40
     TargetMaximum=1
     TargetMinimum=0
     Value=&RigVMFunction_MathVectorDistance(B=@n11.Transform.Translation, A=@n12.Transform.Translation).Result
    ).Result) {variable=pelvisCloseToRoot}
   Maximum = &RigVMFunction_MathDoubleMax(
    B=&RigVMFunction_MathDoubleSub(A=1, B=RigVMVariableNode() {variable=L thighInward alpha}.Value).Result
    A=RigVMVariableNode() {variable=L thighForward alpha}.Value
   )
   > &RigUnit_ParentConstraint(
    AdvancedSettings=(InterpolationType=Average,RotationOrderForFilter=XZY)
    Child=(Type=Bone,Name="dyn_L_frontFlap_master")
    Filter=(TranslationFilter=(bX=true,bY=true,bZ=true),RotationFilter=(bX=True,bY=True,bZ=True),ScaleFilter=(bX=True,bY=True,bZ=True))
    Parents=((Item=(Type=Bone,Name="thigh_l"),Weight=1))
    Weight=@Maximum.Result
   )
   Multiply = &RigVMFunction_MathDoubleMul(B=0.4, A=@Maximum.Result)
   > &RigUnit_ParentConstraint(
    AdvancedSettings=(InterpolationType=Average,RotationOrderForFilter=XZY)
    Child=(Type=Bone,Name="dyn_L_backFlap_master")
    Filter=(TranslationFilter=(bX=false,bY=false,bZ=false),RotationFilter=(bX=True,bY=True,bZ=True),ScaleFilter=(bX=True,bY=True,bZ=True))
    Parents=((Item=(Type=Bone,Name="thigh_l"),Weight=1))
    Weight=@Multiply.Result
   )
   n13 = RigVMVariableNode() {variable=R thighInward alpha}
   > &RigUnit_ParentConstraint(
    AdvancedSettings=(InterpolationType=Average,RotationOrderForFilter=XZY)
    Child=(Type=Bone,Name="dyn_R_frontFlap_master")
    Filter=(TranslationFilter=(bX=true,bY=true,bZ=true),RotationFilter=(bX=True,bY=True,bZ=True),ScaleFilter=(bX=True,bY=True,bZ=True))
    Parents=((Item=(Type=Bone,Name="thigh_r"),Weight=1))
    Weight=&RigVMFunction_MathDoubleMax(B=&RigVMFunction_MathDoubleSub(A=1, B=@n13.Value).Result, A=RigVMVariableNode() {variable=R thighForward alpha}.Value).Result
   )
   n14 = &RigVMFunction_MathDoubleSub(
    A=&RigVMFunction_MathDoubleSub(B=RigVMVariableNode() {variable=spineTwist Rightwards}.Value, A=RigVMVariableNode() {variable=R thighForward alpha}.Value).Result
    B=RigVMVariableNode() {variable=L thighForward alpha}.Value
   )
   n15 = RigVMVariableNode() {variable=eitherFootCloseToPelvis}
   n16 = &RigVMFunction_AlphaInterp(
    InterpSpeedDecreasing=40
    InterpSpeedIncreasing=40
    Scale=0.9
    bClampResult=true
    Value=&RigVMFunction_MathDoubleSub(
     A=&RigVMFunction_MathDoubleSub(A=@n14.Result, B=@n15.Value).Result
     B=RigVMVariableNode() {variable=R thighVeryOutwards alpha}.Value
    ).Result
   )
   > &RigUnit_ParentConstraint(
    AdvancedSettings=(InterpolationType=Average,RotationOrderForFilter=XZY)
    Child=(Type=Bone,Name="dyn_L_frontFlap_master")
    Filter=(TranslationFilter=(bX=true,bY=true,bZ=true),RotationFilter=(bX=True,bY=true,bZ=true),ScaleFilter=(bX=True,bY=True,bZ=True))
    Parents=((Item=(Type=Bone,Name="thigh_r"),Weight=1))
    Weight=@n16.Result
   )
   n17 = &RigVMFunction_MathDoubleSub(
    A=&RigVMFunction_MathDoubleSub(B=RigVMVariableNode() {variable=spineTwist Rightwards}.Value, A=RigVMVariableNode() {variable=L thighForward alpha}.Value).Result
    B=RigVMVariableNode() {variable=R thighForward alpha}.Value
   )
   n18 = RigVMVariableNode() {variable=eitherFootCloseToPelvis}
   n19 = &RigVMFunction_AlphaInterp(
    InterpSpeedDecreasing=40
    InterpSpeedIncreasing=40
    Scale=0.9
    bClampResult=true
    Value=&RigVMFunction_MathDoubleSub(
     A=&RigVMFunction_MathDoubleSub(A=@n17.Result, B=@n18.Value).Result
     B=RigVMVariableNode() {variable=L thighVeryOutwards alpha}.Value
    ).Result
   )
   > &RigUnit_ParentConstraint(
    AdvancedSettings=(InterpolationType=Average,RotationOrderForFilter=XZY)
    Child=(Type=Bone,Name="dyn_R_frontFlap_master")
    Filter=(TranslationFilter=(bX=true,bY=true,bZ=true),RotationFilter=(bX=true,bY=true,bZ=true),ScaleFilter=(bX=True,bY=True,bZ=True))
    Parents=((Item=(Type=Bone,Name="thigh_l"),Weight=1))
    Weight=@n19.Result
   )
   > &RigUnit_ParentConstraint(
    AdvancedSettings=(InterpolationType=Average,RotationOrderForFilter=XZY)
    Child=(Type=Bone,Name="dyn_frontSash_master")
    Filter=(TranslationFilter=(bX=true,bY=true,bZ=true),RotationFilter=(bX=true,bY=true,bZ=true),ScaleFilter=(bX=True,bY=True,bZ=True))
    Parents=((Item=(Type=Bone,Name="thigh_l"),Weight=1))
    Weight=@n19.Result
   )
   > &RigUnit_ParentConstraint(
    AdvancedSettings=(InterpolationType=Average,RotationOrderForFilter=XZY)
    Child=(Type=Bone,Name="dyn_L_front_sleeveFix")
    Filter=(TranslationFilter=(bX=true,bY=true,bZ=true),RotationFilter=(bX=true,bY=true,bZ=true),ScaleFilter=(bX=True,bY=True,bZ=True))
    Parents=((Item=(Type=Bone,Name="upperarm_twist_01_l"),Weight=1))
    Weight=&RigVMFunction_MathDoubleMax(A=RigVMVariableNode() {variable=L upperarmUp alpha}.Value, B=RigVMVariableNode() {variable=L upperarmForward alpha}.Value).Result
   )
   > &RigUnit_ParentConstraint(
    AdvancedSettings=(InterpolationType=Average,RotationOrderForFilter=XZY)
    Child=(Type=Bone,Name="dyn_R_front_sleeveFix")
    Filter=(TranslationFilter=(bX=true,bY=true,bZ=true),RotationFilter=(bX=true,bY=true,bZ=true),ScaleFilter=(bX=True,bY=True,bZ=True))
    Parents=((Item=(Type=Bone,Name="upperarm_twist_01_r"),Weight=1))
    Weight=&RigVMFunction_MathDoubleMax(A=RigVMVariableNode() {variable=R upperarmUp alpha}.Value, B=RigVMVariableNode() {variable=R upperarmForward alpha}.Value).Result
   )
   > &RigUnit_ParentConstraint(
    AdvancedSettings=(InterpolationType=Average,RotationOrderForFilter=XZY)
    Child=(Type=Bone,Name="dyn_L_back_sleeveFix")
    Filter=(TranslationFilter=(bX=true,bY=true,bZ=true),RotationFilter=(bX=true,bY=true,bZ=true),ScaleFilter=(bX=True,bY=True,bZ=True))
    Parents=((Item=(Type=Bone,Name="upperarm_twist_01_l"),Weight=1))
    Weight=&RigVMFunction_MathDoubleMax(B=RigVMVariableNode() {variable=L upperarmUp alpha}.Value, A=RigVMVariableNode() {variable=L upperarmBack alpha}.Value).Result
   )
   > &RigUnit_ParentConstraint(
    AdvancedSettings=(InterpolationType=Average,RotationOrderForFilter=XZY)
    Child=(Type=Bone,Name="dyn_R_back_sleeveFix")
    Filter=(TranslationFilter=(bX=true,bY=true,bZ=true),RotationFilter=(bX=true,bY=true,bZ=true),ScaleFilter=(bX=True,bY=True,bZ=True))
    Parents=((Item=(Type=Bone,Name="upperarm_twist_01_r"),Weight=1))
    Weight=&RigVMFunction_MathDoubleMax(A=RigVMVariableNode() {variable=R upperarmBack alpha}.Value, B=RigVMVariableNode() {variable=R upperarmUp alpha}.Value).Result
   )
   n20 = &RigVMFunction_AlphaInterp(
    InterpSpeedDecreasing=30
    InterpSpeedIncreasing=30
    Scale=1
    bClampResult=False
    Value=&RigVMFunction_MathDoubleMul(B=RigVMVariableNode() {variable=eitherFootCloseToPelvis}.Value, A=RigVMVariableNode() {variable=pelvisCloseToRoot}.Value).Result
   )
   > &RigUnit_ModifyTransforms(
    ItemToModify="((Item=(Type=Bone,Name=\"dyn_L_backFlap_master\"),Transform=(Rotation=(X=0,Y=0,Z=0.8192,W=0.5736),Translation=(X=0,Y=0,Z=0),Scale3D=(X=1,Y=1,Z=1))),(Item=(Type=Bone,Name=\"dyn_R_backFlap_master\"),Transform=(Rotation=(X=0,Y=0,Z=0.766,W=0.6428),Translation=(X=0 ...(truncated)"
    WeightMinimum=0
    Weight=@n20.Result
   )
   &RigVMFunction_MathDoubleRemap(SourceMaximum=45, SourceMinimum=20, TargetMaximum=1, TargetMinimum=0, Value=0)
   &RigVMFunction_MathDoubleRemap(SourceMaximum=-45, SourceMinimum=-20, TargetMaximum=1, TargetMinimum=0, Value=0)
  ]}
]
hierarchy [
 root {loc=<0 0 0>, rot=<0 0 0 1>}
 pelvis {parent=root, loc=<0 -0.5301 98.0571>, rot=<0.0127 0.707 -0.0127 -0.707>}
 spine_01 {parent=pelvis, loc=<6.1208 -0.9328 0>, rot=<0 0 0.034 -0.9994>}
 spine_02 {parent=spine_01, loc=<6.5997 -0.2095 0>, rot=<0 0 0.0238 -0.9997>}
 spine_03 {parent=spine_02, loc=<6.4914 -0.1237 0>, rot=<0 0 0.0149 -0.9999>}
 spine_04 {parent=spine_03, loc=<6.9598 0.0794 0>, rot=<0 0 -0.0132 -0.9999>}
 spine_05 {parent=spine_04, loc=<6.9987 0 0>, rot=<0 0 -0.0015 -1>}
 attach_cape {parent=spine_05, loc=<9.3216 -8.4501 0>, rot=<0.0616 -0.7044 -0.0616 -0.7044>}
 cape {parent=attach_cape, type=Null, loc=<0 3.8435 -3.4606>, rot=<0 0 0 1>}
 neck_01 {parent=spine_05, loc=<11.3138 0.1529 0>, rot=<0 0 -0.2003 -0.9797>}
 neck_02 {parent=neck_01, loc=<4.5627 0 0>, rot=<0 0 0.0504 -0.9987>}
 head {parent=neck_02, loc=<5.3914 0 0>, rot=<0 0 0.115 -0.9934>}
 attach_backpack {parent=spine_05, loc=<0 -8.35 0>, rot=<-0.0093 -0.707 0.0093 -0.707>}
 Backpack_BR {parent=attach_backpack, type=Null, loc=<0 3.6748 0>, rot=<0 0 0 1>}
 clavicle_l {parent=spine_05, loc=<5.4831 -0.1053 -3.2946>, rot=<0.0465 -0.7289 -0.009 -0.6829>}
 upperarm_l {parent=clavicle_l, loc=<9.8658 -1.2136 -0.4105>, rot=<0.0164 -0.3245 -0.0269 -0.9454>}
 lowerarm_l {parent=upperarm_l, loc=<26.465 -0.8345 0.259>, rot=<0.0036 0.0161 -0.2298 -0.9731>}
 hand_l {parent=lowerarm_l, loc=<22.7355 -0.3203 -0.6105>, rot=<-0.6784 -0.0223 0.0735 -0.7307>}
 thumb_01_l {parent=hand_l, loc=<1.3118 -0.3191 -2.459>, rot=<0.6546 -0.4042 0.0664 -0.6354>}
 thumb_02_l {parent=thumb_01_l, loc=<5.9336 -0.0448 0.0401>, rot=<-0.0037 0.0213 0.298 -0.9543>}
 thumb_03_l {parent=thumb_02_l, loc=<2.2223 0.3682 -0.0907>, rot=<0.0118 -0.013 0.0066 -0.9998>}
 index_metacarpal_l {parent=hand_l, loc=<1.6535 0.0183 -2.96>, rot=<0.0158 -0.0124 -0.0173 -0.9996>}
 index_01_l {parent=index_metacarpal_l, loc=<8.3008 0.0357 -0.0524>, rot=<-0.0044 0.0485 0.0924 -0.9945>}
 index_02_l {parent=index_01_l, loc=<3.3587 -0.3004 0.0452>, rot=<-0.002 0.0027 0.3225 -0.9466>}
 index_03_l {parent=index_02_l, loc=<2.5578 -0.123 -0.0252>, rot=<-0.002 0.0067 0.2558 -0.9667>}
 middle_metacarpal_l {parent=hand_l, loc=<2.1856 -0.1725 -0.5984>, rot=<0.006 0.0168 -0.0329 -0.9993>}
 middle_01_l {parent=middle_metacarpal_l, loc=<7.3275 -0.2447 0.0129>, rot=<-0.0208 0.0042 0.0858 -0.9961>}
 middle_02_l {parent=middle_01_l, loc=<4.0123 0.1401 0.1175>, rot=<-0.0218 0.0258 0.3512 -0.9357>}
 middle_03_l {parent=middle_02_l, loc=<2.6077 0.0534 0.1373>, rot=<0.0082 0.0188 0.1708 -0.9851>}
 ring_metacarpal_l {parent=hand_l, loc=<2.2074 -0.3625 1.3638>, rot=<-0.0284 0.0722 -0.0458 -0.9959>}
 ring_01_l {parent=ring_metacarpal_l, loc=<7.0295 -0.6197 -0.1446>, rot=<-0.0489 -0.0252 0.1892 -0.9804>}
 ring_02_l {parent=ring_01_l, loc=<3.6174 0.4647 -0.1027>, rot=<0 0 0.2806 -0.9598>}
 ring_03_l {parent=ring_02_l, loc=<2.4152 0.114 0.0199>, rot=<0 0 0.1189 -0.9929>}
 pinky_metacarpal_l {parent=hand_l, loc=<1.7616 -0.9753 3.3814>, rot=<-0.0397 0.0814 -0.069 -0.9935>}
 pinky_01_l {parent=pinky_metacarpal_l, loc=<7.1709 -0.4831 -0.1094>, rot=<-0.0654 -0.0035 0.2242 -0.9723>}
 pinky_02_l {parent=pinky_01_l, loc=<2.8274 0.0893 -0.1043>, rot=<0.0037 0.0011 0.2854 -0.9584>}
 pinky_03_l {parent=pinky_02_l, loc=<2.2046 0.0307 -0.0078>, rot=<-0.0002 0.0098 0.1445 -0.9895>}
 lowerarm_twist_02_l {parent=lowerarm_l, loc=<10.6707 0 0>, rot=<0 0 0 1>}
 lowerarm_twist_01_l {parent=lowerarm_l, loc=<18.316 0 0>, rot=<0 0 0 1>}
 upperarm_twist_01_l {parent=upperarm_l, loc=<7.2135 -0.1134 0>, rot=<0 0 0 1>}
 upperarm_twist_02_l {parent=upperarm_l, loc=<19.7998 -0.1134 0>, rot=<0 0 0 1>}
 dyn_L_back_sleeveFix {parent=clavicle_l, loc=<9.5251 -1.1559 -0.286>, rot=<0.0123 -0.1806 -0.029 -0.9831>}
 dyn_L_front_sleeveFix {parent=clavicle_l, loc=<9.4731 -0.1238 -0.2895>, rot=<0.0123 -0.1806 -0.029 -0.9831>}
 clavicle_r {parent=spine_05, loc=<5.4831 -0.1053 3.2946>, rot=<0.7289 0.0465 -0.6829 0.009>}
 upperarm_r {parent=clavicle_r, loc=<-9.8658 1.2136 0.4105>, rot=<0.0164 -0.3245 -0.0269 -0.9454>}
 lowerarm_r {parent=upperarm_r, loc=<-26.465 0.8345 -0.259>, rot=<0.0036 0.0161 -0.2298 -0.9731>}
 hand_r {parent=lowerarm_r, loc=<-22.7355 0.3203 0.6106>, rot=<-0.6784 -0.0223 0.0735 -0.7307>}
 thumb_01_r {parent=hand_r, loc=<-1.3118 0.3191 2.459>, rot=<0.6546 -0.4042 0.0664 -0.6354>}
 thumb_02_r {parent=thumb_01_r, loc=<-5.9336 0.0448 -0.0401>, rot=<-0.0037 0.0213 0.298 -0.9543>}
 thumb_03_r {parent=thumb_02_r, loc=<-2.2223 -0.3682 0.0907>, rot=<0.0118 -0.013 0.0066 -0.9998>}
 index_metacarpal_r {parent=hand_r, loc=<-1.6535 -0.0183 2.96>, rot=<0.0158 -0.0124 -0.0173 -0.9996>}
 index_01_r {parent=index_metacarpal_r, loc=<-8.3008 -0.0357 0.0524>, rot=<-0.0044 0.0485 0.0924 -0.9945>}
 index_02_r {parent=index_01_r, loc=<-3.3587 0.3004 -0.0452>, rot=<-0.002 0.0027 0.3225 -0.9466>}
 index_03_r {parent=index_02_r, loc=<-2.5578 0.123 0.0252>, rot=<-0.002 0.0067 0.2558 -0.9667>}
 middle_metacarpal_r {parent=hand_r, loc=<-2.1856 0.1725 0.5984>, rot=<0.006 0.0168 -0.0329 -0.9993>}
 middle_01_r {parent=middle_metacarpal_r, loc=<-7.3275 0.2447 -0.0129>, rot=<-0.0208 0.0042 0.0858 -0.9961>}
 middle_02_r {parent=middle_01_r, loc=<-4.0123 -0.1401 -0.1175>, rot=<-0.0218 0.0258 0.3512 -0.9357>}
 middle_03_r {parent=middle_02_r, loc=<-2.6077 -0.0534 -0.1373>, rot=<0.0082 0.0188 0.1708 -0.9851>}
 ring_metacarpal_r {parent=hand_r, loc=<-2.2074 0.3625 -1.3638>, rot=<-0.0284 0.0722 -0.0458 -0.9959>}
 ring_01_r {parent=ring_metacarpal_r, loc=<-7.0295 0.6197 0.1446>, rot=<-0.0489 -0.0252 0.1892 -0.9804>}
 ring_02_r {parent=ring_01_r, loc=<-3.6174 -0.4647 0.1027>, rot=<0 0 0.2806 -0.9598>}
 ring_03_r {parent=ring_02_r, loc=<-2.4152 -0.114 -0.0199>, rot=<0 0 0.1189 -0.9929>}
 pinky_metacarpal_r {parent=hand_r, loc=<-1.7616 0.9753 -3.3814>, rot=<-0.0397 0.0814 -0.069 -0.9935>}
 pinky_01_r {parent=pinky_metacarpal_r, loc=<-7.1709 0.4831 0.1094>, rot=<-0.0654 -0.0035 0.2242 -0.9723>}
 pinky_02_r {parent=pinky_01_r, loc=<-2.8274 -0.0893 0.1043>, rot=<0.0037 0.0011 0.2854 -0.9584>}
 pinky_03_r {parent=pinky_02_r, loc=<-2.2046 -0.0307 0.0078>, rot=<-0.0002 0.0098 0.1445 -0.9895>}
 lowerarm_twist_02_r {parent=lowerarm_r, loc=<-10.6707 0 0>, rot=<0 0 0 -1>}
 lowerarm_twist_01_r {parent=lowerarm_r, loc=<-18.316 0 0>, rot=<0 0 0 -1>}
 upperarm_twist_01_r {parent=upperarm_r, loc=<-7.2135 0.1134 0>, rot=<0 0 0 -1>}
 upperarm_twist_02_r {parent=upperarm_r, loc=<-19.7998 0.1134 0>, rot=<0 0 0 -1>}
 dyn_R_back_sleeveFix {parent=clavicle_r, loc=<-9.5231 1.233 0.2798>, rot=<0.0123 -0.1806 -0.029 -0.9831>}
 dyn_R_front_sleeveFix {parent=clavicle_r, loc=<-9.4731 0.1238 0.2891>, rot=<0.0123 -0.1806 -0.029 -0.9831>}
 thigh_l {parent=pelvis, loc=<-1.0855 -0.2109 -8.7053>, rot=<0.0009 0.0355 0.0155 -0.9993>}
 calf_l {parent=thigh_l, loc=<-42.5284 -0.0019 -0.7297>, rot=<0.0006 -0.0161 -0.0391 -0.9991>}
 foot_l {parent=calf_l, loc=<-46.7401 -2.5014 -0.592>, rot=<0.0427 -0.0238 0.0353 -0.9982>}
 ball_l {parent=foot_l, loc=<-3.5658 10.8066 -0.1544>, rot=<0.0001 -0.0001 -0.7063 -0.7079>}
 calf_twist_02_l {parent=calf_l, loc=<-19.1546 -1.7943 -0.2638>, rot=<0.0237 0.0095 -0.0035 -0.9997>}
 calf_twist_01_l {parent=calf_l, loc=<-35.8483 -1.9197 -0.5794>, rot=<0.0237 0.0095 -0.0035 -0.9997>}
 thigh_twist_01_l {parent=thigh_l, loc=<-27.952 0 0>, rot=<0.0013 0 0 -1>}
 dyn_side_loopSash {parent=thigh_l, loc=<-49.636 -1.9975 -7.9784>, rot=<-0.0975 -0.6478 0.7518 -0.0753>}
 dyn_L_frontFlap_master {parent=thigh_l, loc=<-3.778 4.7941 -0.5398>, rot=<-0.0112 -0.9983 0.0359 -0.0452>}
 dyn_L_frontFlap_1 {parent=dyn_L_frontFlap_master, loc=<0 0 0>, rot=<0 0 0 1>}
 dyn_L_frontFlap_2 {parent=dyn_L_frontFlap_1, loc=<14.2782 0 0>, rot=<0 0 0 1>}
 dyn_front_loopSash_master {parent=dyn_L_frontFlap_1, loc=<0.5485 2.5062 0.5288>, rot=<-0.1347 0.0795 0.0247 -0.9874>}
 dyn_front_loopSash_1 {parent=dyn_front_loopSash_master, loc=<0 0 0>, rot=<0 0 0 1>}
 dyn_front_loopSash_1_5 {parent=dyn_front_loopSash_1, loc=<11.7872 0.1146 -0.3081>, rot=<-0.0224 -0.015 0.0208 -0.9994>}
 dyn_front_loopSash_2 {parent=dyn_front_loopSash_1_5, loc=<11.7766 0.5983 0.0293>, rot=<-0.0224 -0.015 0.0208 -0.9994>}
 dyn_front_loopSash_3 {parent=dyn_front_loopSash_2, loc=<14.7508 0 0>, rot=<0.0435 0.0822 -0.0036 -0.9957>}
 thigh_r {parent=pelvis, loc=<-1.0855 -0.2109 8.7053>, rot=<0.0355 -0.0009 0.9993 0.0155>}
 calf_r {parent=thigh_r, loc=<42.5284 0.0019 0.7297>, rot=<0.0006 -0.0161 -0.0391 -0.9991>}
 foot_r {parent=calf_r, loc=<46.7401 2.5014 0.592>, rot=<0.0427 -0.0238 0.0353 -0.9982>}
 ball_r {parent=foot_r, loc=<3.5658 -10.8066 0.1544>, rot=<0.0001 -0.0001 -0.7063 -0.7079>}
 calf_twist_02_r {parent=calf_r, loc=<19.1546 1.7943 0.2638>, rot=<0.0237 0.0095 -0.0035 -0.9997>}
 calf_twist_01_r {parent=calf_r, loc=<35.8483 1.9197 0.5794>, rot=<0.0237 0.0095 -0.0035 -0.9997>}
 thigh_twist_01_r {parent=thigh_r, loc=<27.952 0 0>, rot=<0.0013 0 0 -1>}
 dyn_R_frontFlap_master {parent=thigh_r, loc=<1.8545 -5.2662 -2.7098>, rot=<0.9995 -0.0275 -0.0118 -0.0097>}
 dyn_R_frontFlap_1 {parent=dyn_R_frontFlap_master, loc=<0 0 0>, rot=<0 0 0 1>}
 dyn_R_frontFlap_2 {parent=dyn_R_frontFlap_1, loc=<15.0971 0 0>, rot=<0 0.0232 0 -0.9997>}
 dyn_frontSash_master {parent=thigh_r, loc=<-0.885 -7.4804 -0.1641>, rot=<0.9892 -0.0343 0.0085 0.1423>}
 dyn_frontSash_1 {parent=dyn_frontSash_master, loc=<0 0 0>, rot=<0 0 0 1>}
 dyn_frontSash_2 {parent=dyn_frontSash_1, loc=<19.4381 0 0>, rot=<0 0 0 1>}
 dyn_frontSash_3 {parent=dyn_frontSash_2, loc=<17.2552 0 0>, rot=<0 0 0 1>}
 dyn_holster {parent=pelvis, loc=<2.8043 -0.0946 14.0076>, rot=<-0.0083 -0.99 -0.0039 0.1406>}
 dyn_holster_2 {parent=dyn_holster, loc=<8.4517 0 -0.2574>, rot=<0 0 0 1>}
 dyn_R_backFlap_master {parent=pelvis, loc=<-0.142 -6.0328 7.3377>, rot=<0.0491 0.1283 0.99 -0.0308>}
 dyn_R_backFlap_1 {parent=dyn_R_backFlap_master, loc=<0 0 0>, rot=<0 0 0 1>}
 dyn_R_backFlap_2 {parent=dyn_R_backFlap_1, loc=<14.1737 0 0>, rot=<0 0 0 1>}
 dyn_backSash_master {parent=pelvis, loc=<0.8054 -9.3462 7.4828>, rot=<0.0152 0.1344 0.989 -0.0599>}
 dyn_backSash_1 {parent=dyn_backSash_master, loc=<0 0 0>, rot=<0 0 0 1>}
 dyn_backSash_2 {parent=dyn_backSash_1, loc=<19.4004 0 0>, rot=<0 0 0.0424 -0.9991>}
 dyn_backSash_3 {parent=dyn_backSash_2, loc=<19.4363 0 0>, rot=<0 0 0.0175 -0.9998>}
 dyn_backSash_4 {parent=dyn_backSash_3, loc=<20.4104 0 0>, rot=<0 0 0 1>}
 dyn_L_backFlap_master {parent=pelvis, loc=<-1.0075 -5.6639 -7.1469>, rot=<-0.0763 -0.0315 0.9965 -0.0165>}
 dyn_L_backFlap_1 {parent=dyn_L_backFlap_master, loc=<0 0 0>, rot=<0 0 0 1>}
 dyn_L_backFlap_2 {parent=dyn_L_backFlap_1, loc=<12.2242 0 0>, rot=<0 0 0 1>}
 dyn_back_loopSash_master {parent=dyn_L_backFlap_1, loc=<1.8887 5.5478 1.1316>, rot=<0.0613 -0.0615 -0.0503 -0.995>}
 dyn_back_loopSash_1 {parent=dyn_back_loopSash_master, loc=<0 0 0>, rot=<0 0 0 1>}
 dyn_back_loopSash_1_5 {parent=dyn_back_loopSash_1, loc=<11.2762 -1.8943 0.3952>, rot=<0.0676 0.0001 0.0528 -0.9963>}
 dyn_back_loopSash_2 {parent=dyn_back_loopSash_1_5, loc=<11.4155 -0.734 0.2139>, rot=<0.0676 0.0001 0.0528 -0.9963>}
 dyn_back_loopSash_3 {parent=dyn_back_loopSash_2, loc=<20.3209 -1.1025 -2.2579>, rot=<-0.0441 -0.0766 0.0223 -0.9958>}
]
variables [
 L thighForward alpha {type=double, default=0}
 L thighInward alpha {type=double, default=0}
 L thighVeryOutwards alpha {type=double, default=0}
 L upperarmBack alpha {type=double, default=0}
 L upperarmForward alpha {type=double, default=0}
 L upperarmUp alpha {type=double, default=0}
 R thighForward alpha {type=double, default=0}
 R thighInward alpha {type=double, default=0}
 R thighVeryOutwards alpha {type=double, default=0}
 R upperarmBack alpha {type=double, default=0}
 R upperarmForward alpha {type=double, default=0}
 R upperarmUp alpha {type=double, default=0}
 eitherFootCloseToPelvis {type=double, default=0}
 pelvisCloseToRoot {type=double, default=0}
 spineTwist Rightwards {type=double, default=0}
]
