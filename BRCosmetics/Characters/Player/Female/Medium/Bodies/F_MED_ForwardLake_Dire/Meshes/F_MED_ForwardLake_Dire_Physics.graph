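# Physics asset for the F_MED_ForwardLake_Dire female character body, covering its dangling sashes, loop sashes, back and front flaps, and holster.
# Limited swing and twist constraints chain the dyn_ bones from master bones, or from the pelvis for the holster.
# Fully free-rotation constraints cross-link back sash, flaps and both loop-sash ends.
# Leg and pelvis capsules plus a huge offset root capsule provide body shapes; front flaps have no constraints.
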PhysicsAsset /BRCosmetics/Characters/Player/Female/Medium/Bodies/F_MED_ForwardLake_Dire/Meshes/F_MED_ForwardLake_Dire_Physics
bodies {
 calf_l [{shapeType=Capsule, length=50, radius=7, shapeName=thigh_l_capsule, center=<-27.4487 -2.2048 -0.0067>, rotation=<rot 90 0 0>}]
 calf_r [{shapeType=Capsule, length=40, radius=6, shapeName=thigh_l_capsule, center=<27.4487 2.2048 0.0067>, rotation=<rot 90 0 0>}]
 dyn_L_backFlap_1 [
  {shapeType=Box, extentX=20, extentY=2, extentZ=20, shapeName=dyn_L_backFlap_1_box, center=<10 0 0>}
  {
   shapeType=Box
   extentX=20
   extentY=4.5834
   extentZ=3.7222
   shapeName=dyn_R_backFlap_1_box
   center=<9.9256 -2.6908 -6.4809>
   rotation=<rot 3.4705 -4.8336 -30.0609>
  }
 ]
 dyn_L_backFlap_2 [{shapeType=Box, extentX=20, extentY=2, extentZ=20, shapeName=dyn_L_backFlap_2_box, center=<10 0 0>}]
 dyn_L_backFlap_master [{shapeType=Capsule, length=1, radius=1, shapeName=dyn_back_loopSash_master_capsule, rotation=<rot 0 -90 0>}]
 dyn_L_frontFlap_1 [{shapeType=Capsule, length=13, radius=10, shapeName=dyn_R_frontFlap_1_capsule, center=<17.673 -7.9904 0>, rotation=<rot 90 0 0>}]
 dyn_L_frontFlap_2 [{shapeType=Capsule, length=13, radius=10, shapeName=dyn_R_frontFlap_1_capsule, center=<17.673 -7.9904 0>, rotation=<rot 90 0 0>}]
 dyn_R_backFlap_1 [
  {shapeType=Box, extentX=20, extentY=2, extentZ=20, shapeName=dyn_R_backFlap_1_box, center=<10 0.2535 0>, rotation=<rot 0 -5.0001 0>}
  {
   shapeType=Box
   extentX=20
   extentY=4.5834
   extentZ=3.7222
   shapeName=dyn_R_backFlap_1_box
   center=<10 -1.384 8.748>
   rotation=<rot -2.446 -1.4137 30.0302>
  }
 ]
 dyn_R_backFlap_2 [{shapeType=Box, extentX=20, extentY=2.9578, extentZ=20, shapeName=dyn_R_backFlap_2_box, center=<10.0391 -0.5535 0>, rotation=<rot 0 -5 0>}]
 dyn_R_backFlap_master [{shapeType=Capsule, length=1, radius=1, shapeName=dyn_R_backFlap_master_capsule, rotation=<rot 0 -90 0>}]
 dyn_R_frontFlap_1 [{shapeType=Capsule, length=13, radius=8, shapeName=dyn_R_frontFlap_1_capsule, center=<17.673 -7.4549 0>, rotation=<rot 90 0 0>}]
 dyn_backSash_1 [{shapeType=Capsule, length=15, radius=15, shapeName=dyn_backSash_1_capsule, center=<6 14 0>}]
 dyn_backSash_2 [{shapeType=Capsule, length=15, radius=15, shapeName=dyn_backSash_2_capsule, center=<6 12 0>}]
 dyn_backSash_3 [{shapeType=Capsule, length=15, radius=15, shapeName=dyn_backSash_3_capsule, center=<6 8 0>}]
 dyn_backSash_4 [{shapeType=Capsule, length=15, radius=15, shapeName=dyn_backSash_4_capsule, center=<6 10.5 0>}]
 dyn_backSash_master [{shapeType=Capsule, length=1, radius=1, shapeName=dyn_backSash_master_capsule}]
 dyn_back_loopSash_1 [{shapeType=Capsule, length=10, radius=10.3, shapeName=dyn_back_loopSash_1_capsule, center=<4.0589 6 0>, rotation=<rot 90 -60.9787 -55.9786>}]
 dyn_back_loopSash_1_5 [{shapeType=Capsule, length=10, radius=10.3, shapeName=dyn_back_loopSash_1_capsule, center=<4.0589 4.5 0>, rotation=<rot 90 -39.6111 -36.3269>}]
 dyn_back_loopSash_2 [{shapeType=Capsule, length=10, radius=10.3, shapeName=dyn_back_loopSash_2_capsule, center=<4.1075 8 0>, rotation=<rot 90 -25.3927 -30.4876>}]
 dyn_back_loopSash_3 [
  {shapeType=Capsule, length=10, radius=10, shapeName=dyn_front_loopSash_2_capsule, center=<4 6 -5>, rotation=<rot 90 0 0>}
  {shapeType=Capsule, length=20, radius=20, shapeName=dyn_back_loopSash_3_capsule, center=<-0.5304 -6 -24.932>, rotation=<rot 0 0 90>}
 ]
 dyn_back_loopSash_master [{shapeType=Capsule, length=1, radius=1, shapeName=dyn_back_loopSash_master_capsule}]
 dyn_frontSash_1 [{shapeType=Capsule, length=30, radius=15.2, shapeName=dyn_backSash_4_capsule, center=<6 13 0>, rotation=<rot 0 0 -5>}]
 dyn_frontSash_2 [{shapeType=Capsule, length=30, radius=15, shapeName=dyn_backSash_4_capsule, center=<6 12 0>, rotation=<rot 0 0 5>}]
 dyn_frontSash_3 [{shapeType=Capsule, length=30, radius=10, shapeName=dyn_backSash_4_capsule, center=<6 6.5 0>, rotation=<rot 0 0 5>}]
 dyn_frontSash_master [{shapeType=Capsule, length=1, radius=1, shapeName=dyn_frontSash_master_capsule}]
 dyn_front_loopSash_1 [{shapeType=Capsule, length=10, radius=10.3, shapeName=dyn_front_loopSash_1_capsule, center=<3 7.8 0>, rotation=<rot 90 0 0>}]
 dyn_front_loopSash_1_5 [{shapeType=Capsule, length=10, radius=10.3, shapeName=dyn_front_loopSash_1_capsule, center=<3 7.8 0>, rotation=<rot 90 -33.8675 -38.8676>}]
 dyn_front_loopSash_2 [{shapeType=Capsule, length=10, radius=9, shapeName=dyn_front_loopSash_2_capsule, center=<3 5 0>, rotation=<rot 90 0 0>}]
 dyn_front_loopSash_3 [
  {shapeType=Capsule, length=10, radius=7, shapeName=dyn_front_loopSash_2_capsule, center=<3.4888 4.5 2>, rotation=<rot 90 -177.6141 -157.6141>}
  {
   shapeType=Capsule
   length=20
   radius=20
   shapeName=dyn_back_loopSash_3_capsule
   center=<-0.5304 -5.6634 25.8569>
   rotation=<rot 0 0.0001 110>
  }
 ]
 dyn_front_loopSash_master [{shapeType=Capsule, length=1, radius=1, shapeName=dyn_front_loopSash_master_capsule}]
 dyn_holster [{shapeType=Capsule, length=7, radius=4.4, shapeName=dyn_holster_capsule, center=<8 0 -2.4>, rotation=<rot 80 179.9999 179.9998>}]
 dyn_holster_2 [{shapeType=Capsule, length=7, radius=4.4, shapeName=dyn_holster_capsule, center=<8 0 -0.4>, rotation=<rot 80 179.9999 179.9998>}]
 pelvis [{shapeType=Capsule, length=8, radius=4.5, shapeName=pelvis_capsule}]
 root [{shapeType=Capsule, length=100, radius=400, shapeName=root_capsule, center=<0 -15 -400>, rotation=<rot 90 0 90>}]
 thigh_l [{shapeType=Capsule, length=38.4396, radius=5.2885, shapeName=thigh_l_capsule, center=<-20 0 0>, rotation=<rot 90 0 0>}]
 thigh_r [{shapeType=Capsule, length=38.4396, radius=5.2885, shapeName=thigh_l_capsule, center=<20 0 0>, rotation=<rot 90 0 0>}]
}
constraints [
 {bone1Name=dyn_back_loopSash_1, bone2Name=dyn_back_loopSash_master, swing1LimitDegrees=20, swing2LimitDegrees=20, twistLimitDegrees=20}
 {bone1Name=dyn_back_loopSash_2, bone2Name=dyn_back_loopSash_1_5, swing1LimitDegrees=20, swing2LimitDegrees=20, twistLimitDegrees=20}
 {bone1Name=dyn_back_loopSash_3, bone2Name=dyn_back_loopSash_2, swing1LimitDegrees=20, swing2LimitDegrees=20, twistLimitDegrees=20}
 {bone1Name=dyn_front_loopSash_1, bone2Name=dyn_front_loopSash_master, swing1LimitDegrees=20, swing2LimitDegrees=30, twistLimitDegrees=0}
 {bone1Name=dyn_front_loopSash_2, bone2Name=dyn_front_loopSash_1_5, swing1LimitDegrees=20, swing2LimitDegrees=30, twistLimitDegrees=20}
 {bone1Name=dyn_front_loopSash_3, bone2Name=dyn_front_loopSash_2, swing1LimitDegrees=20, swing2LimitDegrees=30, twistLimitDegrees=20}
 {bone1Name=dyn_backSash_2, bone2Name=dyn_backSash_1, swing1LimitDegrees=20, swing2LimitDegrees=60, twistLimitDegrees=20}
 {bone1Name=dyn_backSash_3, bone2Name=dyn_backSash_2, swing1LimitDegrees=20, swing2LimitDegrees=60, twistLimitDegrees=45}
 {bone1Name=dyn_backSash_4, bone2Name=dyn_backSash_3, swing1LimitDegrees=20, swing2LimitDegrees=60, twistLimitDegrees=45}
 {bone1Name=dyn_backSash_1, bone2Name=dyn_backSash_master, swing1LimitDegrees=20, swing2LimitDegrees=40, twistLimitDegrees=0}
 {
  bone1Name=dyn_backSash_master
  bone2Name=dyn_R_backFlap_1
  swing1LimitDegrees=45
  swing1Motion=Free
  swing2LimitDegrees=45
  swing2Motion=Free
  twistLimitDegrees=45
  twistMotion=Free
 }
 {bone1Name=dyn_R_backFlap_1, bone2Name=dyn_R_backFlap_master, swing1LimitDegrees=20, swing2LimitDegrees=10, twistLimitDegrees=10}
 {bone1Name=dyn_R_backFlap_2, bone2Name=dyn_R_backFlap_1, swing1LimitDegrees=20, swing2LimitDegrees=10, twistLimitDegrees=10}
 {bone1Name=dyn_L_backFlap_1, bone2Name=dyn_L_backFlap_master, swing1LimitDegrees=20, swing2LimitDegrees=20, twistLimitDegrees=10}
 {bone1Name=dyn_L_backFlap_2, bone2Name=dyn_L_backFlap_1, swing1LimitDegrees=20, swing2LimitDegrees=20, twistLimitDegrees=10}
 {
  bone1Name=dyn_back_loopSash_1
  bone2Name=dyn_L_backFlap_1
  swing1LimitDegrees=45
  swing1Motion=Free
  swing2LimitDegrees=45
  swing2Motion=Free
  twistLimitDegrees=45
  twistMotion=Free
 }
 {bone1Name=dyn_holster, bone2Name=pelvis, swing1LimitDegrees=10, swing2LimitDegrees=30, twistLimitDegrees=10}
 {bone1Name=dyn_holster_2, bone2Name=dyn_holster, swing1LimitDegrees=10, swing2LimitDegrees=30, twistLimitDegrees=10}
 {
  bone1Name=dyn_backSash_1
  bone2Name=dyn_R_backFlap_1
  swing1LimitDegrees=45
  swing1Motion=Free
  swing2LimitDegrees=45
  swing2Motion=Free
  twistLimitDegrees=45
  twistMotion=Free
 }
 {bone1Name=dyn_frontSash_1, bone2Name=dyn_frontSash_master, swing1LimitDegrees=10, swing2LimitDegrees=30, twistLimitDegrees=10}
 {bone1Name=dyn_frontSash_2, bone2Name=dyn_frontSash_1, swing1LimitDegrees=10, swing2LimitDegrees=30, twistLimitDegrees=10}
 {bone1Name=dyn_frontSash_3, bone2Name=dyn_frontSash_2, swing1LimitDegrees=20, swing2LimitDegrees=30, twistLimitDegrees=10}
 {bone1Name=dyn_back_loopSash_1_5, bone2Name=dyn_back_loopSash_1, swing1LimitDegrees=20, swing2LimitDegrees=20, twistLimitDegrees=20}
 {bone1Name=dyn_front_loopSash_1_5, bone2Name=dyn_front_loopSash_1, swing1LimitDegrees=20, swing2LimitDegrees=30, twistLimitDegrees=10}
 {
  bone1Name=dyn_back_loopSash_3
  bone2Name=dyn_front_loopSash_3
  swing1LimitDegrees=45
  swing1Motion=Free
  swing2LimitDegrees=45
  swing2Motion=Free
  twistLimitDegrees=45
  twistMotion=Free
 }
]
