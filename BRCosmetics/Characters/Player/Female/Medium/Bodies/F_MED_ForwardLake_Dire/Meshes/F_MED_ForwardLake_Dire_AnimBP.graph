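# Post-process style AnimBP for the F_MED_ForwardLake_Dire female character body, animating its sashes, flaps and cloth.
# Copied parent pose runs through Control Rig into two RigidBody sims, blended by a frontend-idle flag.
# Additive bone offsets, constraints and AnimDynamics springs then refine sash, flap and holster bones.
# EventGraph drives Chaos cloth with per-movement-state settings, flips gravity for skydiving, and disables cloth on load for cape/long backblings or custom idle.
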AnimBlueprint /BRCosmetics/Characters/Player/Female/Medium/Bodies/F_MED_ForwardLake_Dire/Meshes/F_MED_ForwardLake_Dire_AnimBP
target_skeleton=&/F_MED_ForwardLake_Dire_Skeleton
graphs [
 AnimGraph {nodes [
   Anim Dynamics - Bone: dyn_R_frontFlap_1!(SimulationSpace:AnimPhysSimSpaceType Enum, ComponentPose, Alpha=1) -> (Pose) {
    angularDampingOverride=0.999
    angularSpringConstant=300
    bAngularSpring=true
    bOverrideAngularDamping=true
    simulationSpace=RootRelative
    alphaBoolBlend {blendOption=Linear}
    boundBone {boneName=dyn_R_frontFlap_1}
    componentAppliedLinearAccClamp=<3000 3000 3000>
    componentLinearAccScale=<1 1 1>
    componentLinearVelScale=<1 1 1>
    physicsBodyDefinitions [{
      boundBone {boneName=dyn_R_frontFlap_1}
      boxExtents=<15 15 15>
      constraintSetup {angularLimitsMax=<0 0 30>, angularLimitsMin=<0 0 3>, angularTarget=<1 0 0>}
      localJointOffset=<3 0 0>
     }]
   }
   n1 = Anim Dynamics - Bone: dyn_frontSash_1!(SimulationSpace:AnimPhysSimSpaceType Enum, ComponentPose, Alpha=1) -> (Pose) {
    angularDampingOverride=0.999
    angularSpringConstant=300
    bAngularSpring=true
    bOverrideAngularDamping=true
    simulationSpace=RootRelative
    alphaBoolBlend {blendOption=Linear}
    boundBone {boneName=dyn_frontSash_1}
    componentAppliedLinearAccClamp=<3000 3000 3000>
    componentLinearAccScale=<1 1 1>
    componentLinearVelScale=<1 1 1>
    physicsBodyDefinitions [{
      boundBone {boneName=dyn_frontSash_1}
      boxExtents=<15 15 15>
      constraintSetup {angularLimitsMax=<0 0 20>, angularLimitsMin=<0 0 0>, angularTarget=<1 0 0>}
      localJointOffset=<3 0 0>
     }]
   }
   Anim Dynamics - Bone: dyn_frontSash_2!(SimulationSpace:AnimPhysSimSpaceType Enum, ComponentPose=@n1, Alpha=1) -> (Pose) {
    angularDampingOverride=0.999
    angularSpringConstant=300
    bAngularSpring=true
    bOverrideAngularDamping=true
    simulationSpace=RootRelative
    alphaBoolBlend {blendOption=Linear}
    boundBone {boneName=dyn_frontSash_2}
    componentAppliedLinearAccClamp=<3000 3000 3000>
    componentLinearAccScale=<1 1 1>
    componentLinearVelScale=<1 1 1>
    physicsBodyDefinitions [{
      boundBone {boneName=dyn_frontSash_2}
      boxExtents=<15 15 15>
      constraintSetup {angularLimitsMax=<0 0 20>, angularLimitsMin=<0 0 0>, angularTarget=<1 0 0>}
      localJointOffset=<3 0 0>
     }]
   }
   "Constraint - dyn_frontSash_1 (PT)"!(ComponentPose, Alpha=1, ConstraintWeights_0=0.5, ConstraintWeights_1:float=0.5) -> (Pose) {
    alphaBoolBlend {blendOption=Linear}
    boneToModify {boneName=dyn_frontSash_1}
    constraintSetup [{targetBone {boneName=dyn_L_frontFlap_1}}, {transformType=Translation, targetBone {boneName=thigh_r}}]
    constraintWeights [1,1]
   }
   n2 = "Constraint - dyn_frontSash_2 (P)"(Alpha=1, ConstraintWeights_0=1) {
    alphaBoolBlend {blendOption=Linear}
    boneToModify {boneName=dyn_frontSash_2}
    constraintSetup [{targetBone {boneName=dyn_R_frontFlap_2}}]
    constraintWeights [1]
   }
   "Constraint - dyn_frontSash_3 (P)"(ComponentPose=@n2, Alpha=1, ConstraintWeights_0=1) {
    alphaBoolBlend {blendOption=Linear}
    boneToModify {boneName=dyn_frontSash_3}
    constraintSetup [{targetBone {boneName=dyn_R_frontFlap_2}}]
    constraintWeights [1]
   }
   n3 = "Transform (Modify) Bone - Bone: dyn_frontSash_master"(ComponentPose=Local To Component(LocalPose=Use cached pose 'RB'()), Alpha=1, Translation="0,1.5,0", Rotation="0,2,0", Scale="1,1,1") {
    rotationMode=BMM_Additive
    rotationSpace=BCS_BoneSpace
    translationMode=BMM_Additive
    translationSpace=BCS_BoneSpace
    alphaBoolBlend {blendOption=Linear}
    boneToModify {boneName=dyn_frontSash_master}
    rotation {yaw=2}
    translation {y=1.5}
   }
   n4 = "Transform (Modify) Bone - Bone: dyn_front_loopSash_master"(ComponentPose=@n3, Alpha=1, Translation="0,1,0", Rotation="0,1,0", Scale="1,1,1") {
    rotationMode=BMM_Additive
    rotationSpace=BCS_BoneSpace
    translationMode=BMM_Additive
    translationSpace=BCS_BoneSpace
    alphaBoolBlend {blendOption=Linear}
    boneToModify {boneName=dyn_front_loopSash_master}
    rotation {yaw=1}
    translation {y=1}
   }
   n5 = "Constraint - dyn_frontSash_master (R)"(ComponentPose=@n4, Alpha=1, ConstraintWeights_0=1) {
    alphaBoolBlend {blendOption=Linear}
    boneToModify {boneName=dyn_frontSash_master}
    constraintSetup [{transformType=Rotation, targetBone {boneName=dyn_R_frontFlap_1}}]
    constraintWeights [1]
   }
   n6 = Anim Dynamics - Bone: dyn_L_frontFlap_1!(SimulationSpace:AnimPhysSimSpaceType Enum, ComponentPose=@n5, Alpha=1) -> (Pose) {
    angularDampingOverride=0.999
    angularSpringConstant=300
    bAngularSpring=true
    bOverrideAngularDamping=true
    simulationSpace=RootRelative
    alphaBoolBlend {blendOption=Linear}
    boundBone {boneName=dyn_L_frontFlap_1}
    componentAppliedLinearAccClamp=<3000 3000 3000>
    componentLinearAccScale=<1 1 1>
    componentLinearVelScale=<1 1 1>
    physicsBodyDefinitions [{
      boundBone {boneName=dyn_L_frontFlap_1}
      boxExtents=<15 15 15>
      constraintSetup {angularLimitsMax=<0 0 20>, angularLimitsMin=<0 -3 3>, angularTarget=<1 0 0>}
      localJointOffset=<3 0 0>
     }]
   }
   n7 = Anim Dynamics - Bone: dyn_L_frontFlap_2!(SimulationSpace:AnimPhysSimSpaceType Enum, ComponentPose=@n6, Alpha=1) -> (Pose) {
    angularDampingOverride=0.999
    angularSpringConstant=300
    bAngularSpring=true
    bOverrideAngularDamping=true
    simulationSpace=RootRelative
    alphaBoolBlend {blendOption=Linear}
    boundBone {boneName=dyn_L_frontFlap_2}
    componentAppliedLinearAccClamp=<3000 3000 3000>
    componentLinearAccScale=<1 1 1>
    componentLinearVelScale=<1 1 1>
    physicsBodyDefinitions [{
      boundBone {boneName=dyn_L_frontFlap_2}
      boxExtents=<15 15 15>
      constraintSetup {angularLimitsMax=<0 0 20>, angularLimitsMin=<0 -3 3>, angularTarget=<1 0 0>}
      localJointOffset=<3 0 0>
     }]
   }
   n8 = "Constraint - dyn_front_loopSash_1 (R)"(ComponentPose=@n7, Alpha=1, ConstraintWeights_0=1) {
    alphaBoolBlend {blendOption=Linear}
    alphaScaleBiasClamp {scale=0.5}
    boneToModify {boneName=dyn_front_loopSash_1}
    constraintSetup [{transformType=Rotation, targetBone {boneName=dyn_L_frontFlap_1}}]
    constraintWeights [1]
   }
   n9 = "Transform (Modify) Bone - Bone: dyn_frontSash_1"!(ComponentPose=@n8, Alpha=1, Translation="0,-1,2", Rotation="0,0,-5") -> (Pose) {
    rotationMode=BMM_Additive
    rotationSpace=BCS_BoneSpace
    translationMode=BMM_Additive
    translationSpace=BCS_BoneSpace
    alphaBoolBlend {blendOption=Linear}
    boneToModify {boneName=dyn_frontSash_1}
    rotation {roll=-5}
    translation {y=-1, z=2}
   }
   n10 = "Transform (Modify) Bone - Bone: dyn_front_loopSash_master"!(ComponentPose=@n9, Alpha=1, Translation="0,0,-1", Rotation="10,0,0") -> (Pose) {
    rotationMode=BMM_Additive
    rotationSpace=BCS_BoneSpace
    translationMode=BMM_Additive
    translationSpace=BCS_BoneSpace
    alphaBoolBlend {blendOption=Linear}
    boneToModify {boneName=dyn_front_loopSash_master}
    rotation {pitch=10}
    translation {z=-1}
   }
   n11 = "Transform (Modify) Bone - Bone: dyn_front_loopSash_1_5"!(ComponentPose=@n10, Alpha=1, Translation="0,0,0", Rotation="-10,0,0") -> (Pose) {
    rotationMode=BMM_Additive
    rotationSpace=BCS_BoneSpace
    translationMode=BMM_Additive
    translationSpace=BCS_BoneSpace
    alphaBoolBlend {blendOption=Linear}
    boneToModify {boneName=dyn_front_loopSash_1_5}
    rotation {pitch=-10}
   }
   n12 = "Transform (Modify) Bone - Bone: dyn_back_loopSash_1"!(ComponentPose=@n11, Alpha=1, Translation="0,0,0", Rotation="0,-5,0") -> (Pose) {
    rotationMode=BMM_Additive
    rotationSpace=BCS_BoneSpace
    translationMode=BMM_Additive
    translationSpace=BCS_BoneSpace
    alphaBoolBlend {blendOption=Linear}
    boneToModify {boneName=dyn_back_loopSash_1}
    rotation {yaw=-5}
   }
   n13 = "Transform (Modify) Bone - Bone: dyn_back_loopSash_2"!(ComponentPose=@n12, Alpha=1, Translation="0,-1,0", Rotation="0,0,0") -> (Pose) {
    rotationMode=BMM_Additive
    rotationSpace=BCS_BoneSpace
    translationMode=BMM_Additive
    translationSpace=BCS_BoneSpace
    alphaBoolBlend {blendOption=Linear}
    boneToModify {boneName=dyn_back_loopSash_2}
    translation {y=-1}
   }
   n14 = "Transform (Modify) Bone - Bone: dyn_back_loopSash_1_5"!(ComponentPose=@n13, Alpha=1, Translation="0,-3,0", Rotation="0,0,0") -> (Pose) {
    rotationMode=BMM_Additive
    rotationSpace=BCS_BoneSpace
    translationMode=BMM_Additive
    translationSpace=BCS_BoneSpace
    alphaBoolBlend {blendOption=Linear}
    boneToModify {boneName=dyn_back_loopSash_1_5}
    translation {y=-3}
   }
   n15 = Anim Dynamics - Bone: dyn_frontSash_1(ComponentPose=@n14, Alpha=1) {
    angularDampingOverride=0.999
    angularSpringConstant=300
    bAngularSpring=true
    bOverrideAngularDamping=true
    simulationSpace=RootRelative
    alphaBoolBlend {blendOption=Linear}
    alphaScaleBiasClamp {scale=0.5}
    boundBone {boneName=dyn_frontSash_1}
    componentAppliedLinearAccClamp=<3000 3000 3000>
    componentLinearAccScale=<1 1 1>
    componentLinearVelScale=<1 1 1>
    physicsBodyDefinitions [{
      boundBone {boneName=dyn_frontSash_1}
      boxExtents=<20 20 20>
      constraintSetup {angularLimitsMax=<0 10 20>, angularLimitsMin=<0 -10 -5>, angularTarget=<1 0 0>}
      localJointOffset=<3 0 0>
     }]
   }
   n16 = Anim Dynamics - Bone: dyn_frontSash_2(ComponentPose=@n15, Alpha=1) {
    angularDampingOverride=0.999
    angularSpringConstant=300
    bAngularSpring=true
    bOverrideAngularDamping=true
    simulationSpace=RootRelative
    alphaBoolBlend {blendOption=Linear}
    alphaScaleBiasClamp {scale=0.5}
    boundBone {boneName=dyn_frontSash_2}
    componentAppliedLinearAccClamp=<3000 3000 3000>
    componentLinearAccScale=<1 1 1>
    componentLinearVelScale=<1 1 1>
    physicsBodyDefinitions [{
      boundBone {boneName=dyn_frontSash_2}
      boxExtents=<20 20 20>
      constraintSetup {angularLimitsMax=<0 10 20>, angularLimitsMin=<0 -10 -5>, angularTarget=<1 0 0>}
      localJointOffset=<3 0 0>
     }]
   }
   n17 = "Constraint - dyn_holster_2 (PPPPPP)"!(
    ComponentPose=@n16
    Alpha=1
    ConstraintWeights_0=0.3632
    ConstraintWeights_1:float=0.3229
    ConstraintWeights_2:float=0.001
    ConstraintWeights_3:float=0.2398
    ConstraintWeights_4:float=0.0305
    ConstraintWeights_5:float=0.0425
   ) -> (Pose) {
    alphaBoolBlend {blendOption=Linear}
    alphaScaleBiasClamp {scale=0.2}
    boneToModify {boneName=dyn_holster_2}
    constraintSetup [
     {targetBone {boneName=pelvis}}
     {targetBone {boneName=thigh_r}}
     {targetBone {boneName=dyn_L_backFlap_1}}
     {targetBone {boneName=dyn_R_backFlap_1}}
     {targetBone {boneName=dyn_R_backFlap_2}}
     {targetBone {boneName=dyn_R_frontFlap_1}}
    ]
    constraintWeights [1,1,1,1,1,1]
   }
   n18 = "Constraint - dyn_backSash_2 (P)"(ComponentPose=@n17, Alpha=1, ConstraintWeights_0=1) {
    alphaBoolBlend {blendOption=Linear}
    boneToModify {boneName=dyn_backSash_2}
    constraintSetup [{targetBone {boneName=thigh_r}}]
    constraintWeights [1]
   }
   n19 = "Constraint - dyn_backSash_3 (P)"(ComponentPose=@n18, Alpha=1, ConstraintWeights_0=1) {
    alphaBoolBlend {blendOption=Linear}
    boneToModify {boneName=dyn_backSash_3}
    constraintSetup [{targetBone {boneName=calf_r}}]
    constraintWeights [1]
   }
   n20 = "Transform (Modify) Bone - Bone: dyn_backSash_2"(ComponentPose=@n19, Alpha=1, Translation="0,5,0", Rotation="-5,0,0", Scale="1,1,1") {
    rotationMode=BMM_Additive
    rotationSpace=BCS_BoneSpace
    translationMode=BMM_Additive
    translationSpace=BCS_BoneSpace
    alphaBoolBlend {blendOption=Linear}
    boneToModify {boneName=dyn_backSash_2}
    rotation {pitch=-5}
    translation {y=5}
   }
   Output Pose(Result=Component To Local(ComponentPose=@n20)) {layerGroup=None}
   n21 = Control Rig(Source=Copy Pose From Mesh() {bUseAttachedParent=true}) {
    alphaBoolBlend {blendOption=Linear}
    controlRigAssetReference {blueprintRigClass=&/F_MED_ForwardLake_Dire_CtrlRig.F_MED_ForwardLake_Dire_CtrlRig_C}
   }
   Save cached pose 'Start'(Pose=@n21) {cacheName=Start}
   n22 = RigidBody!(ComponentPose=Local To Component(LocalPose=Use cached pose 'Start'()), Alpha, OverrideWorldGravity:Vector="0,0,0") -> (Pose) {
    bClampLinearTranslationLimitToRefPose=true
    bOverrideWorldGravity=true
    overridePhysicsAsset=&/F_MED_ForwardLake_Dire_Physics
    simulationSpace=BaseBoneSpace
    alphaBoolBlend {blendOption=Linear}
    baseBoneRef {boneName=pelvis}
    simSpaceSettings {
     dampingAlpha=0
     maxAngularAcceleration=50
     maxAngularVelocity=7
     maxLinearAcceleration=4000
     maxLinearVelocity=800
     velocityScaleZ=0.5
     worldAlpha=1
     externalLinearDragV=<1.5 1.5 1.5>
    }
   }
   n23 = RigidBody(ComponentPose=Local To Component(LocalPose=Use cached pose 'Start'()), OverrideWorldGravity="0,0,0") {
    bClampLinearTranslationLimitToRefPose=true
    bOverrideWorldGravity=true
    overridePhysicsAsset=&/F_MED_ForwardLake_Dire_Physics_FrontendIdles
    simulationSpace=BaseBoneSpace
    alphaBoolBlend {blendOption=Linear}
    alphaScaleBiasClamp {interpSpeedDecreasing=5, interpSpeedIncreasing=5}
    baseBoneRef {boneName=pelvis}
    simSpaceSettings {
     dampingAlpha=0
     maxAngularAcceleration=50
     maxAngularVelocity=7
     maxLinearAcceleration=4000
     maxLinearVelocity=800
     velocityScaleZ=0.5
     worldAlpha=1
     externalLinearDragV=<1.5 1.5 1.5>
    }
   }
   Save cached pose 'RB'(Pose=Blend Poses by bool(
     BlendPose_0=Component To Local(ComponentPose=@n23)
     BlendPose_1=Component To Local(ComponentPose=@n22)
     BlendTime_0=0.5
     BlendTime_1=0.5
     bActiveValue=Get CUsotomIdle() -> (CUsotomIdle:bool)
    ) {blendType=QuarticInOut}) {cacheName=RB}
  ]}
 EventGraph {nodes [
   n1 = Event BlueprintUpdateAnimation()
   > n2 = Sequence!(execute) -> (then_0, then_1, then_2:Exec, then_3:Exec, then_4:Exec)
   > n3 = Branch(Condition=OR Boolean(A=Get bIsInFrontEnd(), B=Get bIsDBNOCarried()))
   > Set simSpaceChange(simSpaceChange:AnimPhysSimSpaceType Enum=World) -> (Output_Get:AnimPhysSimSpaceType Enum)
   Set simSpaceChange(execute=@n3.else)
   n4 = Get bRigidBodyEnabled()
   Set norBAN(execute=@n2.then_4, norBAN:double="To Float (Boolean)"(InBool=NOT Boolean(A=@n4))) -> (Output_Get:double)
   n5 = Branch(execute=@n2.then_3, Condition=Get bIsInFrontEnd())
   n6 = Get bIsPlayingEmote()
   n7 = Set isPlayingFrontendIdle(execute=@n9.then, isPlayingFrontendIdle:float=GetCurveValue(self=GetOwnerFrontEndAnimInstance(), CurveName=Frontend_Idle_Bypass)) -> (Output_Get:float)
   n8 = Branch(execute=@n5.then, Condition=OR Boolean(A=NOT Boolean(A=@n6), B="Equal (Float)"(A=@n7, B=1)))
   > Set CUsotomIdle(CUsotomIdle:bool=true) -> (Output_Get:bool)
   > Set frontendIdling custom(frontendIdling custom:double=1) -> (Output_Get:double)
   > Set frontendIdling OG(frontendIdling OG:double=0) -> (Output_Get:double)
   Set CUsotomIdle(execute=[@n8.else, @n5.else], CUsotomIdle=false)
   > Set frontendIdling custom(frontendIdling custom=0)
   > Set frontendIdling OG(frontendIdling OG=1)
   n9 = Branch(execute=@n2.then_2, Condition=Get bIsInFrontEnd())
   n10 = Cloth Controller Function DEPRECATED!(
    execute=@n2.then_1
    CharacterPartAnimInstance:Custom Character Part Anim Instance Object Reference=Self-Reference()
    ClothingAssetName:String=F_MED_ForwardLake_Dire_Clothing
    Cloth Paramaters:Cloth Params Struct Structure=Get Cloth Paramaters_0() -> (Cloth Paramaters_0:Cloth Params Struct Structure)
    Cloth Core Settings:Cloth Core Setting Struct Structure=Get Cloth Core Settings_0() -> (Cloth Core Settings_0:Cloth Core Setting Struct Structure)
    ClothUpAxis:Vector="1,0,0"
    OnGround_Moving_SpeedLimit:double=2
    OnGround_Moving_AngularVel_Max:double=0.05
    OnGround_Moving_FictitiousVel_Max:double=0.6
    LinVelScale_FwdBwd_Clamp:double=0
    LinVelScale_UpDwn_Clamp:double=0.7
    DeltaTime:double=@n1.DeltaTimeX
    GameTimeInSeconds:double=0
    WindGustsParams:Wind Gust Params Struct Structure
    WindGustLocalSettings:Wind Gust Local Setting Struct Structure
   ) -> (
    then
    CurrentStateOut:String
    Cloth Settings Out:Cloth Setting Struct Structure
    Cloth Core Settings Out:Cloth Core Setting Struct Structure
    WindGustSettings Out:Wind Gust Setting Struct Structure
    WindGustLocalSettings Out:Wind Gust Local Setting Struct Structure
    GustStrength - WindVelocityOut:Vector
    LinearVelVector-DELETE LATER:Vector
   )
   > Set Gust Local Settings(Gust Local Settings:Wind Gust Local Setting Struct Structure=@n10.WindGustLocalSettings Out) -> (Output_Get:Wind Gust Local Setting Struct Structure)
   > SetClothingInteractorIfNotSet()
   n11 = Get bIsSkyDiving()
   > Branch(Condition=AND Boolean(A=Get bIsFalling(), B=NOT Boolean(A=@n11)))
   > SetVelocityScale!(
    execute
    self:Chaos Clothing Interactor Object Reference=Get ChaosClothingInteractor() -> (ChaosClothingInteractor:Chaos Clothing Interactor Object Reference)
    LinearVelocityScale:Vector="0.32,0.32,0.32"
    AngularVelocityScale:float=0.2
    FictitiousAngularScale:float=0
   ) -> (then)
   Event OnCharacterCustomizationLoaded()
   > n12 = Sequence!(execute) -> (then_0, then_1, then_2)
   Set GravityOverride(execute=[@n18.then, @n12.then_0], GravityOverride="0,0,-980")
   n13 = CheckForBackbling()
   n14 = Get frontendIdling custom() -> (frontendIdling custom:double)
   n15 = Branch(execute=@n12.then_2, Condition=OR Boolean(A=OR Boolean(A=@n13.bIsBackblingACape, B=@n13.bIsBackblingLong), B="Equal (Float)"(A=@n14, B=1)))
   n16 = GetPartSkeletalMeshComponent()
   > Set ClothBlendWeight!(execute, ClothBlendWeight:float=0, self:Skeletal Mesh Component Object Reference=@n16) -> (then, Output_Get:float)
   > Set bDisableClothSimulation!(execute, bDisableClothSimulation:bool=true, self:Skeletal Mesh Component Object Reference=@n16) -> (then, Output_Get:bool)
   Set ClothBlendWeight(execute=@n15.else, ClothBlendWeight=1, self=@n16)
   > Set bDisableClothSimulation(bDisableClothSimulation=false, self=@n16)
   Set frontendAlphaScale(execute=@n12.then_1, frontendAlphaScale:double=SelectFloat(A=0.5, B=1, bPickA=Get bIsInFrontEnd())) -> (Output_Get:double)
   > n17 = Branch(Condition=Get bIsInFrontEnd())
   > Set frontendRBANAlpha(frontendRBANAlpha:bool=false) -> (Output_Get:bool)
   > Set nonFrontendRBANAlpha(nonFrontendRBANAlpha:bool=true) -> (Output_Get:bool)
   Set frontendRBANAlpha(execute=@n17.else, frontendRBANAlpha=true)
   > Set nonFrontendRBANAlpha(nonFrontendRBANAlpha=false)
   Event OnBeginSkydiving()
   > Set GravityOverride(GravityOverride="0,0,600")
   n18 = Event OnEndSkydiving()
  ]}
 SetClothingInteractorIfNotSet {nodes [
   SetClothingInteractorIfNotSet() {is=title K2Node_FunctionEntry}
   > Branch(Condition=Get bClothEnabled())
   n1 = MakeLiteralString(Value=F_MED_ForwardLake_Dire_Clothing)
   n2 = GetOwningComponent()
   > n3 = Cast To ChaosClothingInteractor!(
    execute
    Object=GetClothingInteractor!(
     self:Clothing Simulation Interactor Object Reference=GetClothingSimulationInteractor!(self:Skeletal Mesh Component Object Reference=@n2, ClothingAssetName:Name=None) -> (ReturnValue:Clothing Simulation Interactor Object Reference)
     ClothingAssetName:Name=String To Name(InString=@n1)
    ) -> (ReturnValue:Clothing Interactor Object Reference)
   ) -> (then, CastFailed, AsChaos Clothing Interactor:Chaos Clothing Interactor Object Reference) {is=title K2Node_DynamicCast}
   > Set ChaosClothingInteractor(ChaosClothingInteractor:Chaos Clothing Interactor Object Reference=@n3) -> (Output_Get:Chaos Clothing Interactor Object Reference)
  ]}
]
variables [
 ChaosClothingInteractor {type=ChaosClothingInteractor, default=null}
 Cloth Core Settings_0 {
  type=ClothCoreSettingStruct
  default {
   collisionThickness=1.35
   dampingCoefficient=0.2
   frictionCoefficient=0.75
   localDampingCoefficient=0
   numberIterations=2
   numberSubsteps=1
   selfCollisionThickness=0
   animDriveDamping=<0.5 1>
   animDriveStiffness=<0.5 1>
   areaStiffness=<2 2>
   bendingStiffness=<2 2>
   edgeStiffness=<2 2>
   tetherStiffness=<1.1 1.1>
  }
 }
 Cloth Paramaters_0 {
  type=ClothParamsStruct
  default {
   dBNO {
    angularVelocityScale=0.6
    fictitiousAngularScale=0.1
    jointName=pelvis
    dragCoefficient=<0.2 1>
    gravityOverride=<-980 400 0>
    liftCoefficient=<0 0.5>
    linearVelocityScale=<0.3 0.3 0.1>
   }
   dBNO_Carried {
    angularVelocityScale=0.6
    fictitiousAngularScale=0.1
    jointName=pelvis
    dragCoefficient=<0.2 1>
    gravityOverride=<-980 -200 0>
    liftCoefficient=<0 0.5>
    linearVelocityScale=<0.3 0.3 0.1>
   }
   dBNO_Impulse {
    angularVelocityScale=0
    fictitiousAngularScale=0
    jointName=pelvis
    dragCoefficient=<0 0.2>
    gravityOverride=<-3000 -2000 0>
    liftCoefficient=<0 0.1>
    linearVelocityScale=<0.3 0.1 0.3>
   }
   dBNO_RollBwd {
    angularVelocityScale=0
    fictitiousAngularScale=0
    jointName=pelvis
    dragCoefficient=<0 0.2>
    gravityOverride=<-2000 0 0>
    liftCoefficient=<0 0.1>
    linearVelocityScale=<0.3 0.1 0.3>
   }
   dBNO_RollFwd {
    angularVelocityScale=0
    fictitiousAngularScale=0
    jointName=pelvis
    dragCoefficient=<0 0.2>
    gravityOverride=<-980 0 0>
    liftCoefficient=<0 0.1>
    linearVelocityScale=<0.3 0.1 0.3>
   }
   dBNO_Tumbling {
    angularVelocityScale=0
    fictitiousAngularScale=0
    jointName=pelvis
    dragCoefficient=<0 0.2>
    gravityOverride=<-2500 -300 0>
    liftCoefficient=<0 0.1>
    linearVelocityScale=<0.3 0.1 0.3>
   }
   emoteOrMelee {
    angularVelocityScale=0.2
    fictitiousAngularScale=0
    jointName=pelvis
    dragCoefficient=<0.5 2.5>
    gravityOverride=<-980 0 0>
    liftCoefficient=<0.5 0.5>
    linearVelocityScale=<0.25 0.25 0.25>
   }
   front_End {
    angularVelocityScale=0.05
    fictitiousAngularScale=0.6
    jointName=root
    dragCoefficient=<0.2 0.2>
    gravityOverride=<0 -200 -980>
    liftCoefficient=<0.1 0.1>
    linearVelocityScale=<0.1 0.1 0.1>
   }
   onGround_Crouch_Moving {
    angularVelocityScale=0.1
    fictitiousAngularScale=0.3
    jointName=root
    dragCoefficient=<2 2>
    gravityOverride=<0 0 -980>
    liftCoefficient=<0.5 0.5>
    linearVelocityScale=<0.6 0.6 0.5>
   }
   onGround_Falling {
    angularVelocityScale=0.6
    fictitiousAngularScale=0
    jointName=root
    dragCoefficient=<0 0.2>
    gravityOverride=<0 -500 -980>
    liftCoefficient=<0 0.1>
    linearVelocityScale=<2 1 0.3>
   }
   onGround_Moving {
    angularVelocityScale=0.2
    fictitiousAngularScale=0.1
    jointName=root
    dragCoefficient=<1 1.5>
    gravityOverride=<0 0 -980>
    liftCoefficient=<0.5 1>
    linearVelocityScale=<0.4 0.4 0.4>
   }
   onGround_Standing {
    angularVelocityScale=0.2
    fictitiousAngularScale=1.2
    jointName=root
    dragCoefficient=<2.5 2.5>
    gravityOverride=<0 0 -980>
    liftCoefficient=<0.5 0.5>
    linearVelocityScale=<0.6 0.6 0.6>
   }
   rail_Grinding {
    angularVelocityScale=0.2
    fictitiousAngularScale=0
    jointName=root
    dragCoefficient=<0.1 0.3>
    gravityOverride=<50 -500 -400>
    liftCoefficient=<0.1 0.3>
    linearVelocityScale=<0.1 0.1 0.1>
   }
   riding_Moving {
    angularVelocityScale=0.05
    fictitiousAngularScale=1
    jointName=root
    dragCoefficient=<0.5 0.5>
    gravityOverride=<0 0 -980>
    liftCoefficient=<0.5 0.5>
    linearVelocityScale=<0.4 0.6 0.3>
   }
   riding_Moving_Targeting {
    angularVelocityScale=0.05
    fictitiousAngularScale=0.7
    jointName=root
    dragCoefficient=<0.5 0.5>
    gravityOverride=<0 0 -980>
    liftCoefficient=<0.5 0.5>
    linearVelocityScale=<0.2 0.6 0.2>
   }
   riding_Standing {
    angularVelocityScale=0.05
    fictitiousAngularScale=1
    jointName=root
    dragCoefficient=<0.5 0.5>
    gravityOverride=<0 -450 -980>
    liftCoefficient=<0.5 0.5>
    linearVelocityScale=<0.6 0.6 0.3>
   }
   skydive_DivingDown {
    angularVelocityScale=0.1
    fictitiousAngularScale=0
    jointName=pelvis
    dragCoefficient=<0 0.2>
    gravityOverride=<-200 -100 0>
    liftCoefficient=<0 0.2>
    linearVelocityScale=<0.1 0.1 0.1>
   }
   skydive_FloatingDown {
    angularVelocityScale=0.2
    fictitiousAngularScale=0
    jointName=root
    dragCoefficient=<0.2 0.6>
    gravityOverride=<0 0 -980>
    liftCoefficient=<0.2 0.6>
    linearVelocityScale=<0.1 0.12 0.1>
   }
   skydive_GoingUp {
    angularVelocityScale=0.1
    fictitiousAngularScale=0
    jointName=root
    dragCoefficient=<0.01 0.01>
    gravityOverride=<0 -2000 400>
    liftCoefficient=<0.05 0.05>
    linearVelocityScale=<0.1 0.1 0.01>
   }
   skydive_Parachuting {
    angularVelocityScale=0.2
    fictitiousAngularScale=0
    jointName=root
    dragCoefficient=<0.2 0.6>
    gravityOverride=<0 -400 -1500>
    liftCoefficient=<0.2 0.6>
    linearVelocityScale=<0.1 0.12 0.1>
   }
   skydive_Surfing {
    angularVelocityScale=0.6
    fictitiousAngularScale=0.1
    jointName=root
    dragCoefficient=<1.5 1.5>
    gravityOverride=<0 -600 -500>
    liftCoefficient=<1 1>
    linearVelocityScale=<0.1 0.1 0.3>
   }
   swimming {
    angularVelocityScale=0.6
    fictitiousAngularScale=0
    jointName=root
    dragCoefficient=<2.5 2.5>
    gravityOverride=<0 0 0>
    liftCoefficient=<0.1 0.1>
    linearVelocityScale=<0.6 0.6 0.5>
   }
   tactical_Sprint {
    angularVelocityScale=0.4
    fictitiousAngularScale=0.1
    jointName=root
    dragCoefficient=<2 2>
    gravityOverride=<0 0 -980>
    liftCoefficient=<0.5 0.5>
    linearVelocityScale=<0.4 0.4 0.4>
   }
   ziplining {
    angularVelocityScale=0.2
    fictitiousAngularScale=0
    jointName=root
    dragCoefficient=<0.1 0.4>
    gravityOverride=<0 -400 -980>
    liftCoefficient=<0.1 0.4>
    linearVelocityScale=<0.1 0.1 0.1>
   }
  }
 }
 CUsotomIdle {type=boolean}
 frontendAlphaScale {type=number}
 frontendIdling custom {type=number}
 frontendIdling OG {type=number}
 frontendRBANAlpha {type=boolean}
 GravityOverride {type=Vector}
 Gust Local Settings {type=WindGustLocalSettingStruct, default {elapsedTime=0, finalGustStrengthLocal=0, gustDuration=0, gustStrength=0, timeBetweenGusts=0}}
 isPlayingFrontendIdle {type=number}
 nonFrontendRBANAlpha {type=boolean}
 norBAN {type=number}
 simSpaceChange {type=AnimPhysSimSpaceType}
]
