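# Physics asset of cloth colliders for the female medium ForwardLake Dire body, named for self-cloth collision.
# Capsules cover feet, pelvis and lower spine, boxes cover the right-side front and back flap bones, and a large root box sits below the origin like a floor.
# Thigh and calf bodies have no shapes.
# Constraints link the bodies up to root with all angular motion free.
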PhysicsAsset /BRCosmetics/Characters/Player/Female/Medium/Bodies/F_MED_ForwardLake_Dire/Meshes/F_MED_ForwardLake_Dire_SelfClothColliders
bodies {
 calf_l []
 calf_r []
 dyn_R_backFlap_1 [{shapeType=Box, extentX=20, extentY=10, extentZ=40, shapeName=dyn_R_backFlap_1_box, center=<15 -4.0154 0>}]
 dyn_R_backFlap_2 [{shapeType=Box, extentX=20, extentY=10, extentZ=40, shapeName=dyn_R_backFlap_2_box, center=<10.8539 -4.7563 0>, rotation=<rot 0 -5.0001 0>}]
 dyn_R_frontFlap_1 [{shapeType=Box, extentX=20, extentY=10, extentZ=40, shapeName=dyn_R_frontFlap_1_box, center=<15 -2.8503 1.0374>, rotation=<rot 0 0.0001 20>}]
 dyn_R_frontFlap_2 [{
   shapeType=Box
   extentX=19.2596
   extentY=10
   extentZ=40
   shapeName=dyn_R_frontFlap_2_box
   center=<14.999 -3.9959 -0.3655>
   rotation=<rot 1.7082 -4.6997 19.9299>
  }]
 foot_l [{shapeType=Capsule, length=11.0765, radius=8.0344, center=<-4.1811 6.0027 -0.338>, rotation=<rot 0.0388 1.4701 -89.3982>}]
 foot_r [{shapeType=Capsule, length=11.0765, radius=8.0344, center=<4.1811 -6.0027 0.338>, rotation=<rot 0.0388 1.4701 90.6018>}]
 pelvis [{shapeType=Capsule, length=8, radius=7, center=<-2.4471 0 0>}]
 root [{shapeType=Box, extentX=150, extentY=150, extentZ=100, shapeName=root_box, center=<0 0 -50>, rotation=<rot 0 -90 0>}]
 spine_01 [{shapeType=Capsule, length=9, radius=7.5, center=<-5 0 0>}]
 spine_02 [{shapeType=Capsule, length=2.5219, radius=7, center=<-2.6885 0.7223 0>}]
 thigh_l []
 thigh_r []
}
constraints [
 {
  bone1Name=spine_01
  bone2Name=pelvis
  swing1LimitDegrees=45
  swing1Motion=Free
  swing2LimitDegrees=45
  swing2Motion=Free
  twistLimitDegrees=45
  twistMotion=Free
 }
 {
  bone1Name=spine_02
  bone2Name=spine_01
  swing1LimitDegrees=45
  swing1Motion=Free
  swing2LimitDegrees=45
  swing2Motion=Free
  twistLimitDegrees=45
  twistMotion=Free
 }
 {
  bone1Name=thigh_l
  bone2Name=pelvis
  swing1LimitDegrees=45
  swing1Motion=Free
  swing2LimitDegrees=45
  swing2Motion=Free
  twistLimitDegrees=45
  twistMotion=Free
 }
 {
  bone1Name=foot_l
  bone2Name=calf_l
  swing1LimitDegrees=45
  swing1Motion=Free
  swing2LimitDegrees=45
  swing2Motion=Free
  twistLimitDegrees=45
  twistMotion=Free
 }
 {
  bone1Name=thigh_r
  bone2Name=pelvis
  swing1LimitDegrees=45
  swing1Motion=Free
  swing2LimitDegrees=45
  swing2Motion=Free
  twistLimitDegrees=45
  twistMotion=Free
 }
 {
  bone1Name=foot_r
  bone2Name=calf_r
  swing1LimitDegrees=45
  swing1Motion=Free
  swing2LimitDegrees=45
  swing2Motion=Free
  twistLimitDegrees=45
  twistMotion=Free
 }
 {
  bone1Name=calf_r
  bone2Name=thigh_r
  swing1LimitDegrees=45
  swing1Motion=Free
  swing2LimitDegrees=45
  swing2Motion=Free
  twistLimitDegrees=45
  twistMotion=Free
 }
 {
  bone1Name=calf_l
  bone2Name=thigh_l
  swing1LimitDegrees=45
  swing1Motion=Free
  swing2LimitDegrees=45
  swing2Motion=Free
  twistLimitDegrees=45
  twistMotion=Free
 }
 {
  bone1Name=dyn_R_backFlap_1
  bone2Name=pelvis
  swing1LimitDegrees=45
  swing1Motion=Free
  swing2LimitDegrees=45
  swing2Motion=Free
  twistLimitDegrees=45
  twistMotion=Free
 }
 {
  bone1Name=dyn_R_backFlap_2
  bone2Name=dyn_R_backFlap_1
  swing1LimitDegrees=45
  swing1Motion=Free
  swing2LimitDegrees=45
  swing2Motion=Free
  twistLimitDegrees=45
  twistMotion=Free
 }
 {
  bone1Name=dyn_R_frontFlap_1
  bone2Name=thigh_r
  swing1LimitDegrees=45
  swing1Motion=Free
  swing2LimitDegrees=45
  swing2Motion=Free
  twistLimitDegrees=45
  twistMotion=Free
 }
 {
  bone1Name=dyn_R_frontFlap_2
  bone2Name=dyn_R_frontFlap_1
  swing1LimitDegrees=45
  swing1Motion=Free
  swing2LimitDegrees=45
  swing2Motion=Free
  twistLimitDegrees=45
  twistMotion=Free
 }
 {
  bone1Name=pelvis
  bone2Name=root
  swing1LimitDegrees=45
  swing1Motion=Free
  swing2LimitDegrees=45
  swing2Motion=Free
  twistLimitDegrees=45
  twistMotion=Free
 }
]
