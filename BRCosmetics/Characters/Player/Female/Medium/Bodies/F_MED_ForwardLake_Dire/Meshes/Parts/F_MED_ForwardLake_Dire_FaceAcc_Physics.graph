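# Physics asset for the F_MED_ForwardLake_Dire female character's FaceAcc part, defining jointed hair chains and body collision shapes.
# Left and right side-hair chains and a ponytail hang from the head, while two back-hair chains hang from their own master bodies, all with limited swing and twist.
# Capsules cover head, neck, upper spine, clavicles and upper arms.
# An extra constraint with free swing and twist links the ponytail tip to left back hair.
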PhysicsAsset /BRCosmetics/Characters/Player/Female/Medium/Bodies/F_MED_ForwardLake_Dire/Meshes/Parts/F_MED_ForwardLake_Dire_FaceAcc_Physics
bodies {
 clavicle_l [{shapeType=Capsule, length=6, radius=8, shapeName=clavicle_l_capsule, center=<7.7254 0 -3.4528>, rotation=<rot 83 0 0>}]
 clavicle_r [{shapeType=Capsule, length=6, radius=8, shapeName=clavicle_l_capsule, center=<-7.7254 0 3.4528>, rotation=<rot 83 0 0>}]
 dyn_L_hairBack_0 [{shapeType=Capsule, length=6, radius=2.5, shapeName=dyn_L_hairBack_0_capsule, center=<1 0 -2>, rotation=<rot 0 0 90>}]
 dyn_L_hairBack_1 [{shapeType=Capsule, length=6, radius=2.5, shapeName=dyn_L_hairBack_1_capsule, center=<1 0 -0.5>, rotation=<rot 0 0 90>}]
 dyn_L_hairBack_2 [{shapeType=Capsule, length=6, radius=2.5, shapeName=dyn_L_hairBack_2_capsule, center=<1 0 -1.5>, rotation=<rot 0 0 90>}]
 dyn_L_hairBack_3 [{shapeType=Capsule, length=6, radius=2.5, shapeName=dyn_L_hairBack_3_capsule, center=<1 0 -2>, rotation=<rot 0 0 90>}]
 dyn_L_hairBack_4 [
  {shapeType=Capsule, length=6, radius=2.5, shapeName=dyn_L_hairBack_4_capsule, center=<1 0 -2>, rotation=<rot 0.0001 0 75>}
  {
   shapeType=Capsule
   length=2.1415
   radius=2
   shapeName=dyn_R_hairBack_4_capsule
   center=<2.7899 1.0744 -1.2626>
   rotation=<rot 80.0001 -0.0002 4.9999>
  }
 ]
 dyn_L_hairBack_master [{shapeType=Capsule, length=1, radius=1, shapeName=dyn_L_hairBack_master_capsule, rotation=<rot 0 0 90>}]
 dyn_R_hairBack_0 [{shapeType=Capsule, length=6, radius=2.5, shapeName=dyn_R_hairBack_0_capsule, center=<-1 0 2>, rotation=<rot 0 0 90>}]
 dyn_R_hairBack_1 [{shapeType=Capsule, length=6, radius=2.5, shapeName=dyn_R_hairBack_1_capsule, center=<-1 0 0.5>, rotation=<rot 0 0 90>}]
 dyn_R_hairBack_2 [{shapeType=Capsule, length=6, radius=2.5, shapeName=dyn_R_hairBack_2_capsule, center=<-1 0 1.5>, rotation=<rot 0 0 90>}]
 dyn_R_hairBack_3 [{shapeType=Capsule, length=6, radius=2.5, shapeName=dyn_R_hairBack_3_capsule, center=<-1 0 2>, rotation=<rot 0 0 90>}]
 dyn_R_hairBack_4 [
  {shapeType=Capsule, length=6, radius=2.5, shapeName=dyn_R_hairBack_4_capsule, center=<-1 0 2>, rotation=<rot 0.0001 0 75>}
  {
   shapeType=Capsule
   length=2.1415
   radius=2
   shapeName=dyn_R_hairBack_4_capsule
   center=<-2.7534 -0.8062 1.7381>
   rotation=<rot 90 -6.7214 -6.7213>
  }
 ]
 dyn_R_hairBack_master [{shapeType=Capsule, length=1, radius=1, shapeName=dyn_R_hairBack_master_capsule, rotation=<rot 0 0 90>}]
 dyn_hairLeft_0 [{shapeType=Capsule, length=3, radius=2.5, shapeName=dyn_hairLeft_0_capsule, center=<2 0 0>}]
 dyn_hairLeft_1 [{shapeType=Capsule, length=3, radius=2.5, shapeName=dyn_hairLeft_1_capsule, center=<2 0 0>}]
 dyn_hairLeft_2 [{shapeType=Capsule, length=3, radius=2.5, shapeName=dyn_hairLeft_2_capsule, center=<2 0 0>}]
 dyn_hairLeft_3 [{shapeType=Capsule, length=3, radius=2.7, shapeName=dyn_hairLeft_3_capsule, center=<2 0 0>, rotation=<rot 80 0 0>}]
 dyn_hairLeft_4 [{
   shapeType=Box
   extentX=0.7663
   extentY=2.7433
   extentZ=5.0849
   shapeName=dyn_hairLeft_4_box
   center=<4 0 1.6>
   rotation=<rot 59.9999 179.9999 175>
  }]
 dyn_hairRight_0 [{shapeType=Capsule, length=3, radius=2.5, shapeName=dyn_hairRight_0_capsule, center=<-2 0 0>}]
 dyn_hairRight_1 [{shapeType=Capsule, length=3, radius=2.5, shapeName=dyn_hairRight_1_capsule, center=<-2 0 0>}]
 dyn_hairRight_2 [{shapeType=Capsule, length=3, radius=2.5, shapeName=dyn_hairRight_2_capsule, center=<-2 0 0>}]
 dyn_hairRight_3 [{shapeType=Capsule, length=3, radius=2.7, shapeName=dyn_hairRight_3_capsule, center=<-2 0 0>, rotation=<rot -80 0 0>}]
 dyn_hairRight_4 [{shapeType=Capsule, length=4, radius=1, shapeName=dyn_hairRight_3_capsule, center=<-2.7293 0 -1>, rotation=<rot 120 0 0>}]
 dyn_ponytail_1 [{shapeType=Capsule, length=2, radius=3, shapeName=dyn_ponytail_1_capsule, center=<0.7702 0 0>, rotation=<rot 90 0 0>}]
 dyn_ponytail_2 [{shapeType=Capsule, length=2, radius=3, shapeName=dyn_ponytail_2_capsule, center=<0.8306 0 -3.5>, rotation=<rot 90 0 0>}]
 dyn_ponytail_3 [{shapeType=Capsule, length=2, radius=3, shapeName=dyn_ponytail_3_capsule, center=<0.8306 0 -2.5>, rotation=<rot 90 0 0>}]
 head [{shapeType=Capsule, length=1, radius=4, shapeName=head_capsule, center=<2 0 0>, rotation=<rot 90 0 0>}]
 neck_01 [{shapeType=Capsule, length=9, radius=4, shapeName=neck_01_capsule, center=<0 -0.4386 0>, rotation=<rot 90 0 0>}]
 neck_02 [{shapeType=Capsule, length=4, radius=6, shapeName=neck_01_capsule, center=<-1.2945 2.1589 0>, rotation=<rot 90 0 0>}]
 spine_05 [{shapeType=Capsule, length=15, radius=9, shapeName=spine_05_capsule, center=<1.1596 -0.8079 0>}]
 upperarm_l [{shapeType=Capsule, length=30, radius=3, shapeName=upperarm_l_capsule, center=<16 0 0>, rotation=<rot 90 0 0>}]
 upperarm_r [{shapeType=Capsule, length=30, radius=3, shapeName=upperarm_l_capsule, center=<-16 0 0>, rotation=<rot 90 0 0>}]
}
constraints [
 {bone1Name=dyn_hairLeft_0, bone2Name=head, swing1LimitDegrees=10, swing2LimitDegrees=15, twistLimitDegrees=10}
 {bone1Name=dyn_hairLeft_1, bone2Name=dyn_hairLeft_0, swing1LimitDegrees=5, swing2LimitDegrees=30, twistLimitDegrees=10}
 {bone1Name=dyn_hairLeft_2, bone2Name=dyn_hairLeft_1, swing1LimitDegrees=10, swing2LimitDegrees=30, twistLimitDegrees=30}
 {bone1Name=dyn_hairLeft_3, bone2Name=dyn_hairLeft_2, swing1LimitDegrees=10, swing2LimitDegrees=30, twistLimitDegrees=30}
 {bone1Name=dyn_hairRight_0, bone2Name=head, swing1LimitDegrees=10, swing2LimitDegrees=15, twistLimitDegrees=10}
 {bone1Name=dyn_hairRight_1, bone2Name=dyn_hairRight_0, swing1LimitDegrees=5, swing2LimitDegrees=30, twistLimitDegrees=10}
 {bone1Name=dyn_hairRight_2, bone2Name=dyn_hairRight_1, swing1LimitDegrees=10, swing2LimitDegrees=30, twistLimitDegrees=30}
 {bone1Name=dyn_hairRight_3, bone2Name=dyn_hairRight_2, swing1LimitDegrees=10, swing2LimitDegrees=30, twistLimitDegrees=30}
 {bone1Name=dyn_ponytail_1, bone2Name=head, swing1LimitDegrees=30, swing2LimitDegrees=30, twistLimitDegrees=10}
 {bone1Name=dyn_ponytail_2, bone2Name=dyn_ponytail_1, swing1LimitDegrees=30, swing2LimitDegrees=30, twistLimitDegrees=10}
 {bone1Name=dyn_ponytail_3, bone2Name=dyn_ponytail_2, swing1LimitDegrees=30, swing2LimitDegrees=30, twistLimitDegrees=10}
 {bone1Name=dyn_L_hairBack_0, bone2Name=dyn_L_hairBack_master, swing1LimitDegrees=20, swing2LimitDegrees=20, twistLimitDegrees=30}
 {bone1Name=dyn_L_hairBack_1, bone2Name=dyn_L_hairBack_0, swing1LimitDegrees=20, swing2LimitDegrees=20, twistLimitDegrees=30}
 {bone1Name=dyn_L_hairBack_2, bone2Name=dyn_L_hairBack_1, swing1LimitDegrees=20, swing2LimitDegrees=20, twistLimitDegrees=30}
 {bone1Name=dyn_L_hairBack_3, bone2Name=dyn_L_hairBack_2, swing1LimitDegrees=20, swing2LimitDegrees=20, twistLimitDegrees=30}
 {bone1Name=dyn_L_hairBack_4, bone2Name=dyn_L_hairBack_3, swing1LimitDegrees=30, swing2LimitDegrees=30, twistLimitDegrees=30}
 {bone1Name=dyn_R_hairBack_0, bone2Name=dyn_R_hairBack_master, swing1LimitDegrees=20, swing2LimitDegrees=20, twistLimitDegrees=30}
 {bone1Name=dyn_R_hairBack_1, bone2Name=dyn_R_hairBack_0, swing1LimitDegrees=20, swing2LimitDegrees=20, twistLimitDegrees=30}
 {bone1Name=dyn_R_hairBack_2, bone2Name=dyn_R_hairBack_1, swing1LimitDegrees=20, swing2LimitDegrees=20, twistLimitDegrees=30}
 {bone1Name=dyn_R_hairBack_3, bone2Name=dyn_R_hairBack_2, swing1LimitDegrees=20, swing2LimitDegrees=20, twistLimitDegrees=30}
 {bone1Name=dyn_R_hairBack_4, bone2Name=dyn_R_hairBack_3, swing1LimitDegrees=30, swing2LimitDegrees=30, twistLimitDegrees=30}
 {bone1Name=dyn_hairLeft_4, bone2Name=dyn_hairLeft_3, swing1LimitDegrees=10, swing2LimitDegrees=30, twistLimitDegrees=30}
 {bone1Name=dyn_hairRight_4, bone2Name=dyn_hairRight_3, swing1LimitDegrees=10, swing2LimitDegrees=30, twistLimitDegrees=30}
 {
  bone1Name=dyn_ponytail_3
  bone2Name=dyn_L_hairBack_1
  swing1LimitDegrees=45
  swing1Motion=Free
  swing2LimitDegrees=45
  swing2Motion=Free
  twistLimitDegrees=45
  twistMotion=Free
 }
]
